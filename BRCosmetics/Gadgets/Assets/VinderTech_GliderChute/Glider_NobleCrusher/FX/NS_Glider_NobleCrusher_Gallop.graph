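# Niagara system for the Noble Crusher glider's gallop effect, bursting layered red-tinted sprites and meshes from skeletal mesh sockets.
# Looping emitters fire four bursts at frame-based times, with spawn group choosing the socket and SpawnRegular, SpawnForward and bSpawnBack controlling probability or timing.
# Particles partly inherit source movement while curves animate wave, ray, flash, ring and wind layers.
# The visible emitters are disabled, including glows tinted by ForwardDot and RightDot.
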NiagaraSystem /BRCosmetics/Gadgets/Assets/VinderTech_GliderChute/Glider_NobleCrusher/FX/NS_Glider_NobleCrusher_Gallop
dependencies {
 usedDataInterfaces [
  &NiagaraDataInterfaceCurve
  &NiagaraDataInterfaceColorCurve
  &NiagaraDataInterfaceVector2DCurve
  &NiagaraDataInterfaceSkeletalMesh
  &NiagaraDataInterfaceSocketReader
 ]
 usedDynamicInputs [
  &MergeHelper_Vector2Switch
  &SetBoolByFloatComparison
  &Boolean_OrOperation
  &MakeCustomIntFromBool
  &Boolean_AndOperation
 ]
 usedModules [
  &self:SystemSpawnScript.SystemScriptSource.SystemScriptGraph.NiagaraNodeAssignment_1.SetVariables_#1
  &self:SystemSpawnScript.SystemScriptSource.SystemScriptGraph.NiagaraNodeAssignment_2.SetVariables_#2
  &self:SystemSpawnScript.SystemScriptSource.SystemScriptGraph.NiagaraNodeAssignment_0.SetVariables_#3
  &SystemState
  &Timeline
  &self:SystemSpawnScript.SystemScriptSource.SystemScriptGraph.NiagaraNodeAssignment_3.SetVariables_#4
  &self:SystemSpawnScript.SystemScriptSource.SystemScriptGraph.NiagaraNodeAssignment_4.SetVariables_#5
  &self:SystemSpawnScript.SystemScriptSource.SystemScriptGraph.NiagaraNodeAssignment_5.SetVariables_#6
 ]
}
emitters {
 CapsuleGlows001 {
  bEnabled=false
  emitterUpdateScript [
   &EmitterState(Life Cycle Mode=::NewEnumerator1, Loop Behavior=::NewEnumerator1, Loop Duration Mode=::NewEnumerator1, UseLoopDelay=false)
   &SpawnBurst_Instantaneous(Use Spawn Probability=false, UseLoopCountLimit=false)
  ]
  particleSpawnScript [
   &PartitionParticles(ID to Partition:NiagaraInt32=&ReturnExecIndex())
   &InitializeParticle(
    Mesh Scale Mode=::NewEnumerator3
    Mesh Uniform Scale=0.236
    Position Mode=::NewEnumerator0
    Sprite Size Mode=::NewEnumerator3
    Uniform Sprite Size=300
    Write Material Random=true
   )
  ]
  particleUpdateScript [
   &ParticleState(Kill Particles When Lifetime Has Elapsed=false, Let Infinitely Lived Particles Die When Emitter Deactivates=true)
   &SkeletalMeshLocation(
    Bone / Socket Index=4
    Bone Sampling Mode=::NewEnumerator4
    MaskSpawnGroup=false
    Orientation Sampling=::NewEnumerator1
    Sampled Position Offset=<0 243.3301 0>
    Skeletal Mesh=$System.Skeletal Mesh
   )
   &DynamicMaterialParameters(
    Index 0 Param 1=1
    Index 0 Param 2=10
    Module.Index0Param0Write=true
    Module.Index0Param1Write=false
    Module.Index0Param2Write=false
    Module.Index0Param3Write=false
   ) {module=DynamicMaterialParameters001}
   &ScaleColor(
    Curve Index=&RemapRange(Input Value To Remap=$User.ForwardDot, Input Min=-1, Input Max=1, Output Min=0, Output Max=1)
    Linear Color Curve=DataInterface() {propertyValues {
      AlphaCurve {keys [<at 0 0.112>]}
      BlueCurve {keys [<at 0 0>,<at 0.5 1>,<at 1 1>]}
      GreenCurve {keys [<at 0 0>,<at 0.5 1>,<at 1 0>]}
      RedCurve {keys [<at 0 1>,<at 0.5 1>,<at 1 0>]}
     }}
    Scale Mode=2
   )
  ]
  rendererClasses [&NiagaraSpriteRendererProperties]
  renderers [{rendererClass=&NiagaraSpriteRendererProperties}]
 }
 CapsuleGlows002 {
  bEnabled=false
  emitterUpdateScript [
   &EmitterState(Life Cycle Mode=::NewEnumerator1, Loop Behavior=::NewEnumerator1, Loop Duration Mode=::NewEnumerator1, UseLoopDelay=false)
   &SpawnBurst_Instantaneous(Use Spawn Probability=false, UseLoopCountLimit=false)
  ]
  particleSpawnScript [
   &PartitionParticles(ID to Partition:NiagaraInt32=&ReturnExecIndex())
   &InitializeParticle(
    Mesh Scale Mode=::NewEnumerator3
    Mesh Uniform Scale=0.236
    Position Mode=::NewEnumerator0
    Sprite Size Mode=::NewEnumerator3
    Uniform Sprite Size=300
    Write Material Random=true
   )
  ]
  particleUpdateScript [
   &ParticleState(Kill Particles When Lifetime Has Elapsed=false, Let Infinitely Lived Particles Die When Emitter Deactivates=true)
   &SkeletalMeshLocation(
    Bone / Socket Index=4
    Bone Sampling Mode=::NewEnumerator4
    MaskSpawnGroup=false
    Orientation Sampling=::NewEnumerator1
    Sampled Position Offset=<0 -243.3301 0>
    Skeletal Mesh=$System.Skeletal Mesh
   )
   &DynamicMaterialParameters(
    Index 0 Param 1=1
    Index 0 Param 2=10
    Module.Index0Param0Write=true
    Module.Index0Param1Write=false
    Module.Index0Param2Write=false
    Module.Index0Param3Write=false
   ) {module=DynamicMaterialParameters001}
   &ScaleColor(
    Curve Index=&RemapRange(Input Value To Remap=$User.RightDot, Input Min=-1, Input Max=1, Output Min=0, Output Max=1)
    Linear Color Curve=DataInterface() {propertyValues {
      AlphaCurve {keys [<at 0 0.112>]}
      BlueCurve {keys [<at 0 0>,<at 0.5 1>,<at 1 1>]}
      GreenCurve {keys [<at 0 0>,<at 0.5 1>,<at 1 0>]}
      RedCurve {keys [<at 0 1>,<at 0.5 1>,<at 1 0>]}
     }}
    Scale Mode=2
   )
  ]
  rendererClasses [&NiagaraSpriteRendererProperties]
  renderers [{rendererClass=&NiagaraSpriteRendererProperties}]
 }
 CapsuleGlows004 {
  bEnabled=false
  emitterUpdateScript [
   &EmitterState(
    Delay First Loop Only=true
    Life Cycle Mode=::NewEnumerator1
    Loop Delay=&Divide_Float(A=70, B=30)
    Loop Duration=&Divide_Float(A=20, B=30)
    UseLoopDelay=true
   )
   &SpawnBurst_Instantaneous(
    Spawn Probability=$System.SpawnRegular
    Spawn Time=&Divide_Float(A=&MakeCustomFloatFromBool(Bool=$System.bSpawnBack, True Float=9, False Float=1), B=30)
    Use Spawn Probability=true
    UseLoopCountLimit=false
   )
   &SpawnBurst_Instantaneous(
    Spawn Group=1
    Spawn Probability=$System.SpawnRegular
    Spawn Time=&Divide_Float(A=&MakeCustomFloatFromBool(Bool=$System.bSpawnBack, True Float=19, False Float=17), B=30)
    Use Spawn Probability=true
    UseLoopCountLimit=false
   ) {module=SpawnBurst_Instantaneous001}
   &SpawnBurst_Instantaneous(
    Spawn Group=2
    Spawn Probability=$System.SpawnRegular
    Spawn Time=&Divide_Float(A=&MakeCustomFloatFromBool(Bool=$System.bSpawnBack, True Float=14, False Float=13), B=30)
    Use Spawn Probability=true
    UseLoopCountLimit=false
   ) {module=SpawnBurst_Instantaneous002}
   &SpawnBurst_Instantaneous(
    Spawn Group=3
    Spawn Probability=$System.SpawnRegular
    Spawn Time=&Divide_Float(A=&MakeCustomFloatFromBool(Bool=$System.bSpawnBack, True Float=4, False Float=9), B=30)
    Use Spawn Probability=true
    UseLoopCountLimit=false
   ) {module=SpawnBurst_Instantaneous003}
  ]
  particleSpawnScript [
   &PartitionParticles(ID to Partition:NiagaraInt32=&ReturnExecIndex())
   &InitializeParticle(
    Lifetime=0.8
    Mesh Scale Mode=::NewEnumerator3
    Mesh Uniform Scale=0.3
    Position Mode=::NewEnumerator0
    Sprite Rotation Mode=::NewEnumerator2
    Sprite Size Mode=::NewEnumerator3
    Uniform Sprite Size=100
    Write Material Random=true
   )
   &SkeletalMeshLocation(
    Bone / Socket Index=&SpawnGroup()
    MaskSpawnGroup=false
    Mesh Sampling Type=::NewEnumerator5
    Orientation Sampling=::NewEnumerator1
    Skeletal Mesh=$System.Skeletal Mesh
    Socket Sampling Mode=::NewEnumerator5
   )
   &InitialMeshOrientation(
    Mesh Orientation Mode=::NewEnumerator2
    Orientation Coordinate Space=::World
    Rotation=&RandomRangeVector(Minimum=<0 0 0>, Maximum=<0 0 1>)
    Rotation Coordinate Space=::NewEnumerator1
    Use Rotation Vector=true
   )
   &SpriteFacingAndAlignment(
    Alignment Coordinate Space=::World
    Facing Coordinate Space=::World
    Sprite Alignment=&RandomVector(Vector Scale=1)
    Sprite Facing=<0 0 1>
    WriteAlignment=true
    WriteFacing=true
   )
   &AddVelocity(Velocity=<0 -50 0>, Velocity Speed Scale=&FN_FrontEndFloatModifier(Gameplay=1, Front End=1))
  ]
  particleUpdateScript [
   &ParticleState()
   &DynamicMaterialParameters(
    Index 0 Param 1=1
    Index 0 Param 2=10
    Module.Index0Param0Write=true
    Module.Index0Param1Write=false
    Module.Index0Param2Write=false
    Module.Index0Param3Write=false
   ) {enabled=false, module=DynamicMaterialParameters001}
   &ScaleMeshSize(Scale Factor=&VectorFromFloat(Value=&FloatFromCurve(FloatCurve=DataInterface() {propertyValues {Curve {
         defaultValue=1
         keys [
          {arriveTangent=1.1277, interpMode=RCIM_Cubic, leaveTangent=1.1277, tangentMode=RCTM_User, time=0, value=0.2}
          {arriveTangent=0.8268, interpMode=RCIM_Cubic, leaveTangent=0.8268, tangentMode=RCTM_User, time=0.4507, value=0.6528}
          {arriveTangent=0.0224, interpMode=RCIM_Cubic, leaveTangent=0.0224, tangentMode=RCTM_User, time=1, value=1}
         ]
        }}})))
   &DynamicMaterialParameters(
    Index 0 Param 1=&MakeFloatFromVector(Vector=$Particles.Scale, Channel=::NewEnumerator0)
    Index 0 Param 2=2
    Index 0 Param 3=-0.3
    Index 0 Param 4=&RandomRangeFloat(Minimum=3, Maximum=5)
    Index 1 Param 1=1
    Index 1 Param 2=3
    Index 1 Param 3=0.5
    Index 1 Param 4=1
    Index 2 Param 1=2
    Index 2 Param 2=2
    Index 2 Param 3=&FloatFromCurve(
     FloatCurve=DataInterface() {propertyValues {Curve {keys [{interpMode=RCIM_Cubic, leaveTangent=0.3052, tangentMode=RCTM_Break, time=0.35, value=1}, <at 0.5 0>]}}}
     Scale Curve=8
    )
    Module.Index0Float0Write=true
    Module.Index0Float1Write=true
    Module.Index0Float2Write=true
    Module.Index0Float3Write=true
    Module.Index1Float0Write=true
    Module.Index1Float1Write=true
    Module.Index1Float2Write=true
    Module.Index1Float3Write=true
    Module.Index2Float0Write=true
    Module.Index2Float1Write=true
    Module.Index2Float2Write=true
    Module.Index2Float3Write=true
    Write Parameter Index 1=::NewEnumerator1
    Write Parameter Index 2=::NewEnumerator1
   ) {enabled=false}
   &ScaleColor(
    Curve Index=&RemapRange(Input Value To Remap=$User.ForwardDot, Input Min=-1, Input Max=1, Output Min=0, Output Max=1)
    Linear Color Curve=DataInterface() {propertyValues {
      AlphaCurve {keys [<at 0 1>]}
      BlueCurve {keys [<at 0.4325 0>,<at 0.5 1>,<at 0.5691 1>]}
      GreenCurve {keys [<at 0.4325 0>,<at 0.5 1>,<at 0.5691 0>]}
      RedCurve {keys [<at 0.4325 1>,<at 0.5 1>,<at 0.5691 0>]}
     }}
    Scale Mode=2
   ) {enabled=false}
   &ScaleColor(
    Linear Color Curve=DataInterface() {propertyValues {
      AlphaCurve {keys [<at 0 0>,<at 0.0716 0.2487>,<at 0.1723 1>,<at 0.504 0.6985>,<at 0.7379 0.2792>,<at 1 0>]}
      BlueCurve {keys [<at 0 1>,<at 1 1>]}
      GreenCurve {keys [<at 0 1>,<at 1 1>]}
      RedCurve {keys [<at 0 1>,<at 1 1>]}
     }}
    Scale Mode=2
   ) {module=ScaleColor001}
   &InheritSourceMovement(Applied Position Scale=<0.7 0.7 1>, ApplyForce=false, ApplyPosition=true, ApplyPositionFalloff=false)
   &SolveForcesAndVelocity(Clamp Velocity=false, Limit Acceleration=false)
   &SkeletalMeshLocation(
    Bone / Socket Index=&SpawnGroup()
    Bone Sampling Mode=::NewEnumerator4
    MaskSpawnGroup=false
    Orientation Sampling=::NewEnumerator1
    Position Sampling=::NewEnumerator2
    Skeletal Mesh=$System.Skeletal Mesh
    Use Sampled Mesh Attribute as Ribbon ID=false
   ) {enabled=false, module=SkeletalMeshLocation001}
   /SetVariables() {isSetParameters=true, module=SetVariables_#7}
   /SetVariables(
    Particles.ColorRing:LinearColor=&MultiplyLinearColor(
     Linear Color=<rgba 0.8 0.016 0.0683 1>
     Scale Factor=&MakeVector4_FromLinearColor(LinearColor=$Particles.Color)
    )
    Particles.ColorRing2:LinearColor=&MultiplyLinearColor(
     Linear Color=<rgba 2 0 0.1333 0.6174>
     Scale Factor=&MakeVector4_FromLinearColor(LinearColor=$Particles.Color)
    )
    Particles.ExtraScale:Vector3f=&Add_Vector(
     A=$Particles.Scale
     B=&Multiply_VectorByFloat(
      Vector=<0.13 0.13 0>
      Float=&FloatFromCurve(FloatCurve=DataInterface() {propertyValues {Curve {
          defaultValue=1
          keys [
           {arriveTangent=1.1459, interpMode=RCIM_Cubic, leaveTangent=1.1459, tangentMode=RCTM_User, time=0.3853, value=0}
           {arriveTangent=1.705, interpMode=RCIM_Cubic, leaveTangent=1.705, tangentMode=RCTM_User, time=0.6439, value=0.6347}
           {arriveTangent=0.0224, interpMode=RCIM_Cubic, leaveTangent=0.0224, tangentMode=RCTM_User, time=1, value=1}
          ]
         }}})
     )
    )
    Particles.RingsParam0:Vector4f=&MakeVector4(X=1, Y=1, Z=1, W=1)
    Particles.RingsParam1:Vector4f=&MakeVector4(
     X=0
     Y=&FloatFromCurve(
      FloatCurve=DataInterface() {propertyValues {Curve {
         defaultValue=1
         keys [
          {arriveTangent=-0.4912, interpMode=RCIM_Cubic, leaveTangent=-0.4912, tangentMode=RCTM_User, time=0, value=0}
          {arriveTangent=0.0386, interpMode=RCIM_Cubic, leaveTangent=0.0386, tangentMode=RCTM_User, time=0.261, value=0.0195}
          {arriveTangent=2.4695, interpMode=RCIM_Cubic, leaveTangent=2.4695, tangentMode=RCTM_User, time=0.7093, value=0.765}
          {arriveTangent=0.0224, interpMode=RCIM_Cubic, leaveTangent=0.0224, tangentMode=RCTM_User, time=1, value=1}
         ]
        }}}
      Scale Curve=&RandomRangeFloat(Minimum=0.5, Maximum=0.75)
     )
     Z=1
     W=&Add_Float(
      A=&Multiply_Float(A=$Particles.Age, B=0.1)
      B=&RandomRangeFloat(Minimum=0, Maximum=1)
     )
    )
   ) {enabled=false, isSetParameters=true, module=SetVariables_#8}
   /SetVariables(
    Particles.DynamicMaterialParameter_Ripple:Vector4f=&MakeVector4(
     X=&FloatFromCurve(FloatCurve=DataInterface() {propertyValues {Curve {
         defaultValue=0
         keys [
          {arriveTangent=4.1785, interpMode=RCIM_Cubic, leaveTangent=4.1785, tangentMode=RCTM_User, time=0, value=0.2}
          {arriveTangent=1.1179, interpMode=RCIM_Cubic, leaveTangent=1.1179, tangentMode=RCTM_User, time=0.2136, value=0.7}
          {arriveTangent=0.0224, interpMode=RCIM_Cubic, leaveTangent=0.0224, tangentMode=RCTM_User, time=1, value=1}
         ]
        }}})
     Y=&RandomRangeFloat(Minimum=0.1, Maximum=0.7)
     Z=&FloatFromCurve(
      FloatCurve=DataInterface() {propertyValues {Curve {keys [
          {arriveTangent=4.9025, interpMode=RCIM_Cubic, leaveTangent=4.9025, tangentMode=RCTM_User, time=0, value=0}
          {arriveTangent=4.4266, interpMode=RCIM_Cubic, tangentMode=RCTM_Break, time=0.5796, value=2}
         ]}}}
      Scale Curve=&RandomRangeFloat(Minimum=0.025, Maximum=0.1)
     )
     W=&FloatFromCurve(FloatCurve=DataInterface() {propertyValues {Curve {keys [
          {arriveTangent=-3.3445, interpMode=RCIM_Cubic, leaveTangent=-3.3445, tangentMode=RCTM_User, time=0, value=0.462}
          {arriveTangent=-1.1268, interpMode=RCIM_Cubic, leaveTangent=-1.2066, tangentMode=RCTM_Break, time=0.3394, value=-0.3632}
          {arriveTangent=-1.4939, interpMode=RCIM_Cubic, tangentMode=RCTM_Break, time=1.2, value=-1}
         ]}}})
    )
    Particles.DynamicMaterialParameter_Ripple1:Vector4f=&MakeVector4(
     X=&RandomRangeFloat(Minimum=0.5, Maximum=0.8)
     Y=0.5
     Z=0
     W=&RandomRangeFloat(Minimum=1, Maximum=10)
    )
    Particles.RippleColor:LinearColor=&MakeLinearColorFromVectorAndFloat(
     "Vector (RGB)"=<1 1 1>
     "Float (Alpha)"=&FloatFromCurve(
      FloatCurve=DataInterface() {propertyValues {Curve {keys [
          <at 0 1 RCIM_Cubic>
          {arriveTangent=-1, interpMode=RCIM_Cubic, leaveTangent=-1, time=0.6308, value=0.7324}
          {arriveTangent=-1.9182, interpMode=RCIM_Cubic, leaveTangent=-1.9182, tangentMode=RCTM_User, time=1, value=0}
         ]}}}
      Scale Curve=0.01
     )
    )
    Particles.RippleScale:Vector2f=&Vector2DFromFloat(Value=&FloatFromCurve(
      FloatCurve=DataInterface() {propertyValues {Curve {
         defaultValue=1
         keys [
          <at 0 0 RCIM_Cubic>
          {arriveTangent=2.0219, interpMode=RCIM_Cubic, leaveTangent=2.0219, time=0.1205, value=0.5206}
          {arriveTangent=0.5451, interpMode=RCIM_Cubic, leaveTangent=0.5451, time=0.4088, value=0.8265}
          <at 1 1 RCIM_Cubic>
         ]
        }}}
      Scale Curve=43.2494
     ))
    Particles.RippleScale2:Vector2f=&Vector2DFromFloat(Value=&FloatFromCurve(
      FloatCurve=DataInterface() {propertyValues {Curve {
         defaultValue=1
         keys [
          <at 0 0 RCIM_Cubic>
          {arriveTangent=2.0219, interpMode=RCIM_Cubic, leaveTangent=2.0219, time=0.1205, value=0.5206}
          {arriveTangent=0.5451, interpMode=RCIM_Cubic, leaveTangent=0.5451, time=0.4088, value=0.8265}
          <at 1 1 RCIM_Cubic>
         ]
        }}}
      Scale Curve=32.7087
     ))
    Particles.SpriteRotation2:NiagaraFloat=&RandomRangeFloat(Minimum=0, Maximum=1)
   ) {enabled=false, isSetParameters=true}
   /SetVariables(Particles.LerpPosition:NiagaraPosition=&Lerp_Vector(
     A=Linked() {linkedVariable {name=Particles.Position, type=NiagaraPosition}}
     B=Linked() {linkedVariable {name=Particles.SkeletalMeshLocation001.SampledPosition, type=NiagaraPosition}}
     Alpha=&FloatFromCurve(FloatCurve=DataInterface() {propertyValues {Curve {keys [<at 0 1>,<at 0.0179 1>,<at 0.0628 0>,<at 1 0>]}}})
    )) {enabled=false, isSetParameters=true, module=SetVariables_#9}
   /SetVariables(
    Particles.ColorWave:LinearColor=&ColorFromCurve(ColorCurve=DataInterface() {propertyValues {
       AlphaCurve {keys [<at 0 0>,<at 0.1458 0.792>,<at 1 0 RCIM_Cubic>]}
       BlueCurve {keys [<at 0 0.1633 RCIM_Cubic>,<at 0.6071 0.0099>]}
       GreenCurve {keys [<at 0 0.1 RCIM_Cubic>,<at 0.6071 0.0061>]}
       RedCurve {keys [<at 0 2 RCIM_Cubic>,<at 0.6071 0.1217>]}
      }})
    Particles.WaveScale:Vector2f=&Vector2DFromFloat(Value=&FloatFromCurve(
      FloatCurve=DataInterface() {propertyValues {Curve {
         defaultValue=1
         keys [
          <at 0 0 RCIM_Cubic>
          {arriveTangent=2.3128, interpMode=RCIM_Cubic, leaveTangent=2.3128, time=0.018, value=0.5912}
          {arriveTangent=0.4163, interpMode=RCIM_Cubic, leaveTangent=0.4163, time=0.393, value=0.9088}
          <at 1 1 RCIM_Cubic>
         ]
        }}}
      Scale Curve=44.0721
     ))
   ) {isSetParameters=true, module=SetVariables_#10}
   /SetVariables(
    Particles.ColorRays:LinearColor=&ColorFromCurve(ColorCurve=DataInterface() {propertyValues {
       AlphaCurve {keys [<at 0 0>,<at 0.1458 0.229>,<at 0.9944 0>]}
       BlueCurve {keys [{arriveTangent=0.0166, interpMode=RCIM_Cubic, time=0.1661, value=0.0667}]}
       GreenCurve {keys [{arriveTangent=0.0101, interpMode=RCIM_Cubic, time=0.1661, value=0}]}
       RedCurve {keys [{arriveTangent=0.2029, interpMode=RCIM_Cubic, time=0.1661, value=2}]}
      }})
    Particles.RaysScale:Vector2f=&Vector2DFromFloat(Value=&FloatFromCurve(
      FloatCurve=DataInterface() {propertyValues {Curve {
         defaultValue=1
         keys [
          <at 0 0 RCIM_Cubic>
          {arriveTangent=0.4651, interpMode=RCIM_Cubic, leaveTangent=0.4651, tangentMode=RCTM_User, time=0.0146, value=0.6971}
          {arriveTangent=0.3074, interpMode=RCIM_Cubic, leaveTangent=0.3074, time=0.393, value=0.9088}
          <at 1 1 RCIM_Cubic>
         ]
        }}}
      Scale Curve=71.4351
     ))
   ) {isSetParameters=true, module=SetVariables_#11}
   /SetVariables(
    Particles.ColorFlash:LinearColor=&ColorFromCurve(ColorCurve=DataInterface() {propertyValues {
       AlphaCurve {keys [<at 0 0>,<at 0.1559 1>,<at 0.5882 0.1>,<at 0.8421 0>]}
       BlueCurve {keys [<at 0.0181 2.6096>, {arriveTangent=-3.6639, interpMode=RCIM_Cubic, leaveTangent=-3.6639, time=0.0689, value=0.3838}, <at 0.6405 0.3294>]}
       GreenCurve {keys [<at 0.0181 0.6588>, {arriveTangent=-0.4935, interpMode=RCIM_Cubic, leaveTangent=-0.4935, time=0.0689, value=0.3294}, <at 0.6405 0.3516>]}
       RedCurve {keys [<at 0.0181 6>, {arriveTangent=-4.8205, interpMode=RCIM_Cubic, leaveTangent=-4.8205, time=0.0689, value=3}, <at 0.6405 3>]}
      }})
    Particles.FlashScale:Vector2f=&Vector2DFromCurve(
     Vector2Curve=DataInterface() {propertyValues {
       XCurve {keys [{arriveTangent=1.6105, interpMode=RCIM_Cubic, leaveTangent=1.6105, tangentMode=RCTM_User, time=0, value=0.3937}, <at 1 1 RCIM_Cubic>]}
       YCurve {keys [
         <at 0 0.2877 RCIM_Cubic>
         {arriveTangent=1.3271, interpMode=RCIM_Cubic, leaveTangent=1.3271, time=0.0463, value=1}
         {arriveTangent=-0.5243, interpMode=RCIM_Cubic, leaveTangent=-0.5243, time=0.16, value=0.5}
         <at 1 0.5 RCIM_Cubic>
        ]}
      }}
     ScaleCurve=<90 40>
    )
   ) {isSetParameters=true, module=SetVariables_#12}
   /SetVariables(
    Particles.ColorNebula:LinearColor=&MultiplyLinearColor(Linear Color=$Particles.Color, Scale Factor=<1 1 1 0.15>)
    Particles.NebulaScale:Vector2f=<70 70>
   ) {isSetParameters=true, module=SetVariables_#13}
   /SetVariables(
    Particles.ExtraScale:Vector3f=&Multiply_Vector(A=$Particles.Scale, B=<0.7 0.7 1>)
    Particles.MeshOrientation2:Quat4f=&OrientMeshToVector(
     Look At Direction=&RandomRangeVector(Minimum=<-1 -1 0>, Maximum=<1 1 0>)
     Reference Axis Vector=<1 0 0>
    )
    Particles.WindColor:LinearColor=&ColorFromCurve(ColorCurve=DataInterface() {propertyValues {
       AlphaCurve {keys [
         <at 0 0.664>
         {arriveTangent=-0.8348, interpMode=RCIM_Cubic, leaveTangent=-0.8348, time=0.0602, value=1.2}
         <at 0.758 0.0313>
         <at 0.9451 0 RCIM_Cubic>
        ]}
       BlueCurve {keys [<at 0 1.3565>,<at 0.4954 0 RCIM_Cubic>]}
       GreenCurve {keys [<at 0 1.3565>,<at 0.4954 0 RCIM_Cubic>]}
       RedCurve {keys [<at 0 3>,<at 0.4954 1 RCIM_Cubic>]}
      }})
   ) {isSetParameters=true, module=SetVariables_#14}
  ]
  rendererClasses [&NiagaraSpriteRendererProperties, &NiagaraMeshRendererProperties]
  renderers [
   {rendererClass=&NiagaraSpriteRendererProperties}
   {rendererClass=&NiagaraMeshRendererProperties, rendererIndex=1}
   {rendererClass=&NiagaraMeshRendererProperties, rendererIndex=2}
   {rendererClass=&NiagaraSpriteRendererProperties, rendererIndex=3}
   {rendererClass=&NiagaraSpriteRendererProperties, rendererIndex=4}
   {rendererClass=&NiagaraSpriteRendererProperties, rendererIndex=5}
   {rendererClass=&NiagaraSpriteRendererProperties, rendererIndex=6}
   {rendererClass=&NiagaraSpriteRendererProperties, rendererIndex=7}
   {rendererClass=&NiagaraSpriteRendererProperties, rendererIndex=8}
   {rendererClass=&NiagaraSpriteRendererProperties, rendererIndex=9}
   {rendererClass=&NiagaraSpriteRendererProperties, rendererIndex=10}
   {rendererClass=&NiagaraSpriteRendererProperties, rendererIndex=11}
   {rendererClass=&NiagaraMeshRendererProperties, rendererIndex=12}
   {rendererClass=&NiagaraMeshRendererProperties, rendererIndex=13}
  ]
 }
 CapsuleGlows005 {
  bEnabled=false
  emitterUpdateScript [
   &EmitterState(
    Delay First Loop Only=true
    Life Cycle Mode=::NewEnumerator1
    Loop Delay=&Divide_Float(A=70, B=30)
    Loop Duration=&Divide_Float(A=40, B=90)
    UseLoopDelay=true
   )
   &SpawnBurst_Instantaneous(
    Spawn Probability=&ClampFloat(Float=&Subtract_Float(A=$System.SpawnForward, B=$System.SpawnRegular))
    Spawn Time=&Divide_Float(A=2, B=90)
    Use Spawn Probability=true
    UseLoopCountLimit=false
   )
   &SpawnBurst_Instantaneous(
    Spawn Group=1
    Spawn Probability=&ClampFloat(Float=&Subtract_Float(A=$System.SpawnForward, B=$System.SpawnRegular))
    Spawn Time=&Divide_Float(A=34, B=90)
    Use Spawn Probability=true
    UseLoopCountLimit=false
   ) {module=SpawnBurst_Instantaneous001}
   &SpawnBurst_Instantaneous(
    Spawn Group=2
    Spawn Probability=&ClampFloat(Float=&Subtract_Float(A=$System.SpawnForward, B=$System.SpawnRegular))
    Spawn Time=&Divide_Float(A=26, B=90)
    Use Spawn Probability=true
    UseLoopCountLimit=false
   ) {module=SpawnBurst_Instantaneous002}
   &SpawnBurst_Instantaneous(
    Spawn Group=3
    Spawn Probability=&ClampFloat(Float=&Subtract_Float(A=$System.SpawnForward, B=$System.SpawnRegular))
    Spawn Time=&Divide_Float(A=18, B=90)
    Use Spawn Probability=true
    UseLoopCountLimit=false
   ) {module=SpawnBurst_Instantaneous003}
  ]
  particleSpawnScript [
   &PartitionParticles(ID to Partition:NiagaraInt32=&ReturnExecIndex())
   &InitializeParticle(
    Lifetime=0.8
    Mesh Scale Mode=::NewEnumerator3
    Mesh Uniform Scale=0.18
    Position Mode=::NewEnumerator0
    Sprite Size Mode=::NewEnumerator3
    Uniform Sprite Size=100
    Write Material Random=true
   )
   &SkeletalMeshLocation(
    Bone / Socket Index=&SpawnGroup()
    MaskSpawnGroup=false
    Mesh Sampling Type=::NewEnumerator5
    Orientation Sampling=::NewEnumerator1
    Skeletal Mesh=$System.Skeletal Mesh
    Socket Sampling Mode=::NewEnumerator5
   )
   &InitialMeshOrientation(Mesh Orientation Mode=::NewEnumerator2, Orientation Coordinate Space=::World, Rotation=<0.204 0 0>, Use Rotation Vector=false)
  ]
  particleUpdateScript [
   &ParticleState()
   &DynamicMaterialParameters(
    Index 0 Param 1=1
    Index 0 Param 2=10
    Module.Index0Param0Write=true
    Module.Index0Param1Write=false
    Module.Index0Param2Write=false
    Module.Index0Param3Write=false
   ) {module=DynamicMaterialParameters001}
   &ScaleMeshSize(Scale Factor=&VectorFromFloat(Value=&FloatFromCurve(FloatCurve=DataInterface() {propertyValues {Curve {
         defaultValue=1
         keys [
          {arriveTangent=1.1277, interpMode=RCIM_Cubic, leaveTangent=1.1277, tangentMode=RCTM_User, time=0, value=0.2}
          {arriveTangent=0.8268, interpMode=RCIM_Cubic, leaveTangent=0.8268, tangentMode=RCTM_User, time=0.4507, value=0.6528}
          {arriveTangent=0.0224, interpMode=RCIM_Cubic, leaveTangent=0.0224, tangentMode=RCTM_User, time=1, value=1}
         ]
        }}})))
   &DynamicMaterialParameters(
    Index 0 Param 1=&MakeFloatFromVector(Vector=$Particles.Scale, Channel=::NewEnumerator0)
    Index 0 Param 2=2
    Index 0 Param 3=-0.3
    Index 0 Param 4=&RandomRangeFloat(Minimum=3, Maximum=5)
    Index 1 Param 1=1
    Index 1 Param 2=3
    Index 1 Param 3=0.5
    Index 1 Param 4=1
    Index 2 Param 1=2
    Index 2 Param 2=2
    Index 2 Param 3=&FloatFromCurve(
     FloatCurve=DataInterface() {propertyValues {Curve {keys [{interpMode=RCIM_Cubic, leaveTangent=0.3052, tangentMode=RCTM_Break, time=0.35, value=1}, <at 0.5 0>]}}}
     Scale Curve=8
    )
    Module.Index0Float0Write=true
    Module.Index0Float1Write=true
    Module.Index0Float2Write=true
    Module.Index0Float3Write=true
    Module.Index1Float0Write=true
    Module.Index1Float1Write=true
    Module.Index1Float2Write=true
    Module.Index1Float3Write=true
    Module.Index2Float0Write=true
    Module.Index2Float1Write=true
    Module.Index2Float2Write=true
    Module.Index2Float3Write=true
    Write Parameter Index 1=::NewEnumerator1
    Write Parameter Index 2=::NewEnumerator1
   ) {enabled=false}
   &ScaleColor(
    Curve Index=&RemapRange(Input Value To Remap=$User.ForwardDot, Input Min=-1, Input Max=1, Output Min=0, Output Max=1)
    Linear Color Curve=DataInterface() {propertyValues {
      AlphaCurve {keys [<at 0 1>]}
      BlueCurve {keys [<at 0.4325 0>,<at 0.5 1>,<at 0.5691 1>]}
      GreenCurve {keys [<at 0.4325 0>,<at 0.5 1>,<at 0.5691 0>]}
      RedCurve {keys [<at 0.4325 1>,<at 0.5 1>,<at 0.5691 0>]}
     }}
    Scale Mode=2
   ) {enabled=false}
   &ScaleColor(
    Linear Color Curve=DataInterface() {propertyValues {
      AlphaCurve {keys [<at 0 0>,<at 0.0716 0.2487>,<at 0.1723 1>,<at 0.504 0.6985>,<at 0.7379 0.2792>,<at 1 0>]}
      BlueCurve {keys [<at 0 1>,<at 1 1>]}
      GreenCurve {keys [<at 0 1>,<at 1 1>]}
      RedCurve {keys [<at 0 1>,<at 1 1>]}
     }}
    Scale Mode=2
   ) {module=ScaleColor001}
   &InheritSourceMovement(
    Applied Position Scale=&Add_Vector(A=<0.6 0.6 0>, B=<0 0 1>)
    ApplyForce=false
    ApplyPosition=true
    ApplyPositionFalloff=false
   )
   &InheritSourceMovement(
    Applied Position Scale=&Add_Vector(
     A=&Multiply_VectorByFloat(
      Vector=<1 1 0>
      Float=&FloatFromCurve(FloatCurve=DataInterface() {propertyValues {Curve {keys [<at 0 1>,<at 0.6969 0>,<at 1 0>]}}}, Scale Curve=0.6)
     )
     B=<0 0 1>
    )
    ApplyForce=false
    ApplyPosition=true
    ApplyPositionFalloff=false
   ) {enabled=false, module=InheritSourceMovement001}
   &SolveForcesAndVelocity(Clamp Velocity=false, Limit Acceleration=false)
   &SkeletalMeshLocation(
    Bone / Socket Index=&SpawnGroup()
    Bone Sampling Mode=::NewEnumerator4
    MaskSpawnGroup=false
    Orientation Sampling=::NewEnumerator1
    Position Sampling=::NewEnumerator2
    Skeletal Mesh=$System.Skeletal Mesh
    Use Sampled Mesh Attribute as Ribbon ID=false
   ) {enabled=false, module=SkeletalMeshLocation001}
   /SetVariables(
    Particles.ColorRing:LinearColor=&MultiplyLinearColor(
     Linear Color=<rgba 2 0.04 0.1707 1>
     Scale Factor=&MakeVector4_FromLinearColor(LinearColor=$Particles.Color)
    )
    Particles.ColorRing2:LinearColor=&MultiplyLinearColor(
     Linear Color=<rgba 2 0 0.1333 0.6174>
     Scale Factor=&MakeVector4_FromLinearColor(LinearColor=$Particles.Color)
    )
    Particles.EmptyParams:Vector4f=&MakeVector4(X=0, Y=0, Z=0, W=0)
    Particles.ExtraScale:Vector3f=&Add_Vector(
     A=$Particles.Scale
     B=&Multiply_VectorByFloat(
      Vector=<0.13 0.13 0>
      Float=&FloatFromCurve(FloatCurve=DataInterface() {propertyValues {Curve {
          defaultValue=1
          keys [
           {arriveTangent=1.1459, interpMode=RCIM_Cubic, leaveTangent=1.1459, tangentMode=RCTM_User, time=0.3853, value=0}
           {arriveTangent=1.705, interpMode=RCIM_Cubic, leaveTangent=1.705, tangentMode=RCTM_User, time=0.6439, value=0.6347}
           {arriveTangent=0.0224, interpMode=RCIM_Cubic, leaveTangent=0.0224, tangentMode=RCTM_User, time=1, value=1}
          ]
         }}})
     )
    )
    Particles.RingsParam0:Vector4f=&MakeVector4(X=1, Y=1, Z=1, W=1)
    Particles.RingsParam1:Vector4f=&MakeVector4(
     X=0
     Y=&FloatFromCurve(
      FloatCurve=DataInterface() {propertyValues {Curve {
         defaultValue=1
         keys [
          {arriveTangent=-0.4912, interpMode=RCIM_Cubic, leaveTangent=-0.4912, tangentMode=RCTM_User, time=0, value=0}
          {arriveTangent=0.0386, interpMode=RCIM_Cubic, leaveTangent=0.0386, tangentMode=RCTM_User, time=0.261, value=0.0195}
          {arriveTangent=2.4695, interpMode=RCIM_Cubic, leaveTangent=2.4695, tangentMode=RCTM_User, time=0.7093, value=0.765}
          {arriveTangent=0.0224, interpMode=RCIM_Cubic, leaveTangent=0.0224, tangentMode=RCTM_User, time=1, value=1}
         ]
        }}}
      Scale Curve=&RandomRangeFloat(Minimum=0.5, Maximum=0.75)
     )
     Z=1
     W=&Add_Float(
      A=&Multiply_Float(A=$Particles.Age, B=0.1)
      B=&RandomRangeFloat(Minimum=0, Maximum=1)
     )
    )
   ) {isSetParameters=true, module=SetVariables_#8}
   /SetVariables(Particles.LerpPosition:NiagaraPosition=&Lerp_Vector(
     A=Linked() {linkedVariable {name=Particles.Position, type=NiagaraPosition}}
     B=Linked() {linkedVariable {name=Particles.SkeletalMeshLocation001.SampledPosition, type=NiagaraPosition}}
     Alpha=&FloatFromCurve(FloatCurve=DataInterface() {propertyValues {Curve {keys [<at 0 1>,<at 0.0179 1>,<at 0.0628 0>,<at 1 0>]}}})
    )) {enabled=false, isSetParameters=true, module=SetVariables_#9}
  ]
  rendererClasses [&NiagaraSpriteRendererProperties, &NiagaraMeshRendererProperties]
  renderers [
   {rendererClass=&NiagaraSpriteRendererProperties}
   {rendererClass=&NiagaraMeshRendererProperties, rendererIndex=1}
   {rendererClass=&NiagaraMeshRendererProperties, rendererIndex=2}
  ]
 }
 CapsuleGlows008 {
  bEnabled=false
  emitterUpdateScript [
   &EmitterState(
    Delay First Loop Only=true
    Life Cycle Mode=::NewEnumerator1
    Loop Delay=&Divide_Float(A=70, B=30)
    Loop Duration=&Divide_Float(A=40, B=90)
    UseLoopDelay=true
   )
   &SpawnBurst_Instantaneous(
    Spawn Group=3
    Spawn Probability=&ClampFloat(Float=&Subtract_Float(A=$System.SpawnForward, B=$System.SpawnRegular))
    Spawn Time=&Divide_Float(A=18, B=90)
    Use Spawn Probability=true
    UseLoopCountLimit=false
   ) {module=SpawnBurst_Instantaneous003}
   &SpawnBurst_Instantaneous(
    Spawn Group=2
    Spawn Probability=&ClampFloat(Float=&Subtract_Float(A=$System.SpawnForward, B=$System.SpawnRegular))
    Spawn Time=&Divide_Float(A=26, B=90)
    Use Spawn Probability=true
    UseLoopCountLimit=false
   ) {module=SpawnBurst_Instantaneous002}
   &SpawnBurst_Instantaneous(
    Spawn Group=1
    Spawn Probability=&ClampFloat(Float=&Subtract_Float(A=$System.SpawnForward, B=$System.SpawnRegular))
    Spawn Time=&Divide_Float(A=34, B=90)
    Use Spawn Probability=true
    UseLoopCountLimit=false
   ) {module=SpawnBurst_Instantaneous001}
   &SpawnBurst_Instantaneous(
    Spawn Probability=&ClampFloat(Float=&Subtract_Float(A=$System.SpawnForward, B=$System.SpawnRegular))
    Spawn Time=&Divide_Float(A=2, B=90)
    Use Spawn Probability=true
    UseLoopCountLimit=false
   )
  ]
  particleSpawnScript [
   &PartitionParticles(ID to Partition:NiagaraInt32=&ReturnExecIndex())
   &InitializeParticle(
    Lifetime=0.8
    Mesh Scale Mode=::NewEnumerator3
    Mesh Uniform Scale=0.4
    Position Mode=::NewEnumerator0
    Sprite Rotation Mode=::NewEnumerator2
    Sprite Size Mode=::NewEnumerator3
    Uniform Sprite Size=100
    Write Material Random=true
   )
   &SkeletalMeshLocation(
    Bone / Socket Index=&SpawnGroup()
    MaskSpawnGroup=false
    Mesh Sampling Type=::NewEnumerator5
    Orientation Sampling=::NewEnumerator1
    Skeletal Mesh=$System.Skeletal Mesh
    Socket Sampling Mode=::NewEnumerator5
   )
   &InitialMeshOrientation(
    Mesh Orientation Mode=::NewEnumerator2
    Orientation Coordinate Space=::World
    Rotation=&RandomRangeVector(Minimum=<0 0 0>, Maximum=<0 0 1>)
    Rotation Coordinate Space=::NewEnumerator1
    Use Rotation Vector=true
   )
   &SpriteFacingAndAlignment(
    Alignment Coordinate Space=::World
    Facing Coordinate Space=::World
    Sprite Alignment=&RandomVector(Vector Scale=1)
    Sprite Facing=<0 0 1>
    WriteAlignment=true
    WriteFacing=true
   )
   &AddVelocity(Velocity=<0 -50 0>, Velocity Speed Scale=&FN_FrontEndFloatModifier(Gameplay=0, Front End=1))
  ]
  particleUpdateScript [
   &ParticleState()
   &ScaleMeshSize(Scale Factor=&VectorFromFloat(Value=&FloatFromCurve(FloatCurve=DataInterface() {propertyValues {Curve {
         defaultValue=1
         keys [
          {arriveTangent=1.1277, interpMode=RCIM_Cubic, leaveTangent=1.1277, tangentMode=RCTM_User, time=0, value=0.2}
          {arriveTangent=0.8268, interpMode=RCIM_Cubic, leaveTangent=0.8268, tangentMode=RCTM_User, time=0.4507, value=0.6528}
          {arriveTangent=0.0224, interpMode=RCIM_Cubic, leaveTangent=0.0224, tangentMode=RCTM_User, time=1, value=1}
         ]
        }}})))
   &ScaleColor(
    Linear Color Curve=DataInterface() {propertyValues {
      AlphaCurve {keys [<at 0 0>,<at 0.0716 0.2487>,<at 0.1723 1>,<at 0.504 0.6985>,<at 0.7379 0.2792>,<at 1 0>]}
      BlueCurve {keys [<at 0 1>,<at 1 1>]}
      GreenCurve {keys [<at 0 1>,<at 1 1>]}
      RedCurve {keys [<at 0 1>,<at 1 1>]}
     }}
    Scale Mode=2
   ) {module=ScaleColor001}
   &InheritSourceMovement(Applied Position Scale=<0.7 0.7 1>, ApplyForce=false, ApplyPosition=true, ApplyPositionFalloff=false)
   &SolveForcesAndVelocity(Clamp Velocity=false, Limit Acceleration=false)
   /SetVariables() {isSetParameters=true, module=SetVariables_#7}
   /SetVariables(
    Particles.ColorWave:LinearColor=&ColorFromCurve(ColorCurve=DataInterface() {propertyValues {
       AlphaCurve {keys [<at 0 0>,<at 0.1458 1.3>,<at 1 0 RCIM_Cubic>]}
       BlueCurve {keys [<at 0 0.1633 RCIM_Cubic>,<at 0.6071 0.0099>]}
       GreenCurve {keys [<at 0 0.1 RCIM_Cubic>,<at 0.6071 0.0061>]}
       RedCurve {keys [<at 0 2 RCIM_Cubic>,<at 0.6071 0.1217>]}
      }})
    Particles.WaveScale:Vector2f=&Vector2DFromFloat(Value=&FloatFromCurve(
      FloatCurve=DataInterface() {propertyValues {Curve {
         defaultValue=1
         keys [
          <at 0 0 RCIM_Cubic>
          {arriveTangent=2.3128, interpMode=RCIM_Cubic, leaveTangent=2.3128, time=0.018, value=0.5912}
          {arriveTangent=0.4163, interpMode=RCIM_Cubic, leaveTangent=0.4163, time=0.393, value=0.9088}
          <at 1 1 RCIM_Cubic>
         ]
        }}}
      Scale Curve=62.4812
     ))
   ) {isSetParameters=true, module=SetVariables_#10}
   /SetVariables(
    Particles.ColorRays:LinearColor=&ColorFromCurve(ColorCurve=DataInterface() {propertyValues {
       AlphaCurve {keys [<at 0 0>,<at 0.1458 0.14>,<at 0.9944 0>]}
       BlueCurve {keys [{arriveTangent=0.0166, interpMode=RCIM_Cubic, time=0.1661, value=0.0667}]}
       GreenCurve {keys [{arriveTangent=0.0101, interpMode=RCIM_Cubic, time=0.1661, value=0}]}
       RedCurve {keys [{arriveTangent=0.2029, interpMode=RCIM_Cubic, time=0.1661, value=2}]}
      }})
    Particles.RaysScale:Vector2f=&Vector2DFromFloat(Value=&FloatFromCurve(
      FloatCurve=DataInterface() {propertyValues {Curve {
         defaultValue=1
         keys [
          <at 0 0 RCIM_Cubic>
          {arriveTangent=0.4651, interpMode=RCIM_Cubic, leaveTangent=0.4651, tangentMode=RCTM_User, time=0.0146, value=0.6971}
          {arriveTangent=0.3074, interpMode=RCIM_Cubic, leaveTangent=0.3074, time=0.393, value=0.9088}
          <at 1 1 RCIM_Cubic>
         ]
        }}}
      Scale Curve=94.1286
     ))
   ) {isSetParameters=true, module=SetVariables_#11}
   /SetVariables(
    Particles.ColorFlash:LinearColor=&ColorFromCurve(ColorCurve=DataInterface() {propertyValues {
       AlphaCurve {keys [<at 0 0>,<at 0.1559 1>,<at 0.391 0.1>,<at 0.6898 0>]}
       BlueCurve {keys [{arriveTangent=-3.9747, interpMode=RCIM_Cubic, time=0.0355, value=0.2527}, <at 0.4374 0.0783>]}
       GreenCurve {keys [{arriveTangent=-0.9116, interpMode=RCIM_Cubic, time=0.0355, value=0.1565}, <at 0.4374 0.0783>]}
       RedCurve {keys [{arriveTangent=-4.7107, interpMode=RCIM_Cubic, time=0.0355, value=2}, <at 0.4374 3>]}
      }})
    Particles.FlashScale:Vector2f=&Vector2DFromCurve(
     Vector2Curve=DataInterface() {propertyValues {
       XCurve {keys [{arriveTangent=1.6105, interpMode=RCIM_Cubic, leaveTangent=1.6105, tangentMode=RCTM_User, time=0, value=0.3937}, <at 1 1 RCIM_Cubic>]}
       YCurve {keys [
         <at 0 0.2877 RCIM_Cubic>
         {arriveTangent=1.3271, interpMode=RCIM_Cubic, leaveTangent=1.3271, time=0.0463, value=1}
         {arriveTangent=-0.5243, interpMode=RCIM_Cubic, leaveTangent=-0.5243, time=0.16, value=0.5}
         <at 1 0.5 RCIM_Cubic>
        ]}
      }}
     ScaleCurve=<70 80>
    )
   ) {isSetParameters=true, module=SetVariables_#12}
   /SetVariables(
    Particles.ColorNebula:LinearColor=&MultiplyLinearColor(Linear Color=$Particles.Color, Scale Factor=<0.7 0.7 0.7 0.3>)
    Particles.NebulaScale:Vector2f=<80 80>
   ) {isSetParameters=true, module=SetVariables_#13}
   /SetVariables(
    Particles.ExtraScale:Vector3f=&Multiply_Vector(A=$Particles.Scale, B=<0.7 0.7 1>)
    Particles.MeshOrientation2:Quat4f=&OrientMeshToVector(
     Look At Direction=&RandomRangeVector(Minimum=<-1 -1 0>, Maximum=<1 1 0>)
     Reference Axis Vector=<1 0 0>
    )
    Particles.WindColor:LinearColor=&ColorFromCurve(ColorCurve=DataInterface() {propertyValues {
       AlphaCurve {keys [
         <at 0 0.664>
         {arriveTangent=-0.8348, interpMode=RCIM_Cubic, leaveTangent=-0.8348, time=0.0602, value=1.2}
         <at 0.758 0.0313>
         <at 0.9451 0 RCIM_Cubic>
        ]}
       BlueCurve {keys [<at 0 0.4696>,<at 0.4954 0.1696 RCIM_Cubic>]}
       GreenCurve {keys [<at 0 0.4696>,<at 0.4954 0.1696 RCIM_Cubic>]}
       RedCurve {keys [<at 0 2>,<at 0.4954 1.5 RCIM_Cubic>]}
      }})
   ) {isSetParameters=true, module=SetVariables_#14}
  ]
  rendererClasses [&NiagaraSpriteRendererProperties, &NiagaraMeshRendererProperties]
  renderers [
   {rendererClass=&NiagaraSpriteRendererProperties}
   {rendererClass=&NiagaraSpriteRendererProperties, rendererIndex=1}
   {rendererClass=&NiagaraSpriteRendererProperties, rendererIndex=2}
   {rendererClass=&NiagaraSpriteRendererProperties, rendererIndex=3}
   {rendererClass=&NiagaraMeshRendererProperties, rendererIndex=4}
   {rendererClass=&NiagaraMeshRendererProperties, rendererIndex=5}
  ]
 }
 CapsuleGlows009 {
  bEnabled=false
  emitterUpdateScript [
   &EmitterState(
    Delay First Loop Only=true
    Life Cycle Mode=::NewEnumerator1
    Loop Delay=&Divide_Float(A=70, B=30)
    Loop Duration=&Divide_Float(A=40, B=90)
    UseLoopDelay=true
   )
   &SpawnBurst_Instantaneous(
    Spawn Count=&RandomRangeInt(Minimum=3, Maximum=5)
    Spawn Group=3
    Spawn Probability=&ClampFloat(Float=&Subtract_Float(A=$System.SpawnForward, B=$System.SpawnRegular))
    Spawn Time=&Divide_Float(A=18, B=90)
    Use Spawn Probability=true
    UseLoopCountLimit=false
   ) {module=SpawnBurst_Instantaneous003}
   &SpawnBurst_Instantaneous(
    Spawn Count=&RandomRangeInt(Minimum=3, Maximum=5)
    Spawn Group=2
    Spawn Probability=&ClampFloat(Float=&Subtract_Float(A=$System.SpawnForward, B=$System.SpawnRegular))
    Spawn Time=&Divide_Float(A=26, B=90)
    Use Spawn Probability=true
    UseLoopCountLimit=false
   ) {module=SpawnBurst_Instantaneous002}
   &SpawnBurst_Instantaneous(
    Spawn Count=&RandomRangeInt(Minimum=3, Maximum=5)
    Spawn Group=1
    Spawn Probability=&ClampFloat(Float=&Subtract_Float(A=$System.SpawnForward, B=$System.SpawnRegular))
    Spawn Time=&Divide_Float(A=34, B=90)
    Use Spawn Probability=true
    UseLoopCountLimit=false
   ) {module=SpawnBurst_Instantaneous001}
   &SpawnBurst_Instantaneous(
    Spawn Count=&RandomRangeInt(Minimum=3, Maximum=5)
    Spawn Probability=&ClampFloat(Float=&Subtract_Float(A=$System.SpawnForward, B=$System.SpawnRegular))
    Spawn Time=&Divide_Float(A=2, B=90)
    Use Spawn Probability=true
    UseLoopCountLimit=false
   )
  ]
  particleSpawnScript [
   &PartitionParticles(ID to Partition:NiagaraInt32=&ReturnExecIndex())
   &InitializeParticle(
    Color=&RandomRangeLinearColor(Minimum=<rgba 4 0.2435 0.2435 0.2435>, Maximum=<rgba 5 0.3478 0.7355 0.713>)
    Lifetime=0.8
    Position Mode=::NewEnumerator0
    Sprite Size Mode=::NewEnumerator1
    Uniform Sprite Size Max=32
    Uniform Sprite Size Min=9
    Write Material Random=true
   )
   &SkeletalMeshLocation(
    Bone / Socket Index=&SpawnGroup()
    MaskSpawnGroup=false
    Mesh Sampling Type=::NewEnumerator5
    Orientation Sampling=::NewEnumerator1
    Skeletal Mesh=$System.Skeletal Mesh
    Socket Sampling Mode=::NewEnumerator5
   )
   &ShapeLocation(
    MaskSpawnGroup=false
    Override Seed=false
    Ring Radius=13
    Rotation Coordinate Space=::World
    Shape Primitive=::NewEnumerator4
    UseOwnerScale=true
   )
   &AddVelocity(
    Rotation Coordinate Space=::World
    Velocity=&AddVectorToPosition(
     Position=&RandomVector(Vector Scale=0.5)
     Vector=&RandomRangeVector(Minimum=<0 0 0.6>, Maximum=<0 0 0.8>)
    )
    Velocity Speed Scale=70
   ) {module=AddVelocity001}
   &AddVelocity(Velocity=<0 -50 0>, Velocity Speed Scale=&FN_FrontEndFloatModifier(Gameplay=0, Front End=1))
  ]
  particleUpdateScript [
   &ParticleState()
   &Drag(Drag=1.5, Rotational Drag=1.5, Use Linear Drag=true, Use Rotational Drag=true)
   &ScaleSpriteSize(
    Scale Sprite Size Mode=::NewEnumerator0
    Uniform Scale Factor=&Multiply_Float(
     A=&FloatFromCurve(FloatCurve=DataInterface() {propertyValues {Curve {
         defaultValue=1
         keys [
          {arriveTangent=6.4214, interpMode=RCIM_Cubic, leaveTangent=6.4214, tangentMode=RCTM_User, time=0, value=0}
          {interpMode=RCIM_Cubic, tangentMode=RCTM_User, time=0.2882, value=1}
          {arriveTangent=-2.7349, tangentMode=RCTM_User, time=1, value=0}
         ]
        }}})
     B=&FloatFromCurve(
      FloatCurve=DataInterface() {propertyValues {Curve {keys [<at 0 0.2>,<at 0.1 0.8647>,<at 0.7 1>,<at 0.85 1.3>]}}}
      CurveIndex=&RandomRangeFloat(Minimum=0, Maximum=1, Evaluation Type=::NewEnumerator1)
     )
    )
   )
   &ScaleColor(
    Linear Color Curve=DataInterface() {propertyValues {
      AlphaCurve {keys [<at 0 0>,<at 0.0716 0.2487>,<at 0.1723 1>,<at 0.504 0.6985>,<at 0.7379 0.2792>,<at 1 0>]}
      BlueCurve {keys [<at 0 1>,<at 1 1>]}
      GreenCurve {keys [<at 0 1>,<at 1 1>]}
      RedCurve {keys [<at 0 1>,<at 1 1>]}
     }}
    Scale Mode=2
   ) {module=ScaleColor001}
   &InheritSourceMovement(Applied Position Scale=<0.7 0.7 1>, ApplyForce=false, ApplyPosition=true, ApplyPositionFalloff=false)
   &SolveForcesAndVelocity(Clamp Velocity=false, Limit Acceleration=false)
  ]
  rendererClasses [&NiagaraSpriteRendererProperties]
  renderers [{rendererClass=&NiagaraSpriteRendererProperties}]
 }
 Flares_Fwd001 {
  bEnabled=false
  emitterUpdateScript [
   &EmitterState(
    Delay First Loop Only=true
    Life Cycle Mode=::NewEnumerator1
    Loop Delay=&Divide_Float(A=70, B=30)
    Loop Duration=&Divide_Float(A=20, B=30)
    UseLoopDelay=true
   )
   &SpawnBurst_Instantaneous(
    Spawn Probability=&ClampFloat(Float=&Subtract_Float(A=$System.SpawnForward, B=$System.SpawnRegular))
    Spawn Time=&Divide_Float(A=2, B=90)
    Use Spawn Probability=true
    UseLoopCountLimit=false
   )
   &SpawnBurst_Instantaneous(
    Spawn Group=1
    Spawn Probability=&ClampFloat(Float=&Subtract_Float(A=$System.SpawnForward, B=$System.SpawnRegular))
    Spawn Time=&Divide_Float(A=34, B=90)
    Use Spawn Probability=true
    UseLoopCountLimit=false
   ) {module=SpawnBurst_Instantaneous001}
   &SpawnBurst_Instantaneous(
    Spawn Group=2
    Spawn Probability=&ClampFloat(Float=&Subtract_Float(A=$System.SpawnForward, B=$System.SpawnRegular))
    Spawn Time=&Divide_Float(A=26, B=90)
    Use Spawn Probability=true
    UseLoopCountLimit=false
   ) {module=SpawnBurst_Instantaneous002}
   &SpawnBurst_Instantaneous(
    Spawn Group=3
    Spawn Probability=&ClampFloat(Float=&Subtract_Float(A=$System.SpawnForward, B=$System.SpawnRegular))
    Spawn Time=&Divide_Float(A=18, B=90)
    Use Spawn Probability=true
    UseLoopCountLimit=false
   ) {module=SpawnBurst_Instantaneous003}
  ]
  particleSpawnScript [
   &PartitionParticles(ID to Partition:NiagaraInt32=&ReturnExecIndex()) {enabled=false}
   &InitializeParticle(
    Lifetime=0.8
    Mesh Scale Mode=::NewEnumerator3
    Mesh Uniform Scale=0.4
    Position Mode=::NewEnumerator0
    Sprite Rotation Mode=::NewEnumerator2
    Sprite Size Mode=::NewEnumerator3
    Uniform Sprite Size=100
    Write Material Random=true
   )
   &SkeletalMeshLocation(
    Bone / Socket Index=&SpawnGroup()
    MaskSpawnGroup=false
    Mesh Sampling Type=::NewEnumerator5
    Orientation Sampling=::NewEnumerator1
    Skeletal Mesh=$System.Skeletal Mesh
    Socket Sampling Mode=::NewEnumerator5
   )
   &InitialMeshOrientation(
    Mesh Orientation Mode=::NewEnumerator2
    Orientation Coordinate Space=::World
    Rotation=&RandomRangeVector(Minimum=<0 0 0>, Maximum=<0 0 1>)
    Rotation Coordinate Space=::NewEnumerator1
    Use Rotation Vector=true
   )
   &SpriteFacingAndAlignment(
    Alignment Coordinate Space=::World
    Facing Coordinate Space=::World
    Sprite Alignment=&RandomVector(Vector Scale=1)
    Sprite Facing=<0 0 1>
    WriteAlignment=true
    WriteFacing=true
   )
   &AddVelocity(Velocity=<0 -50 0>, Velocity Speed Scale=&FN_FrontEndFloatModifier(Gameplay=0, Front End=1))
  ]
  particleUpdateScript [
   &ParticleState()
   &ScaleMeshSize(Scale Factor=&VectorFromFloat(Value=&FloatFromCurve(FloatCurve=DataInterface() {propertyValues {Curve {
         defaultValue=1
         keys [
          {arriveTangent=1.1277, interpMode=RCIM_Cubic, leaveTangent=1.1277, tangentMode=RCTM_User, time=0, value=0.2}
          {arriveTangent=0.8268, interpMode=RCIM_Cubic, leaveTangent=0.8268, tangentMode=RCTM_User, time=0.4507, value=0.6528}
          {arriveTangent=0.0224, interpMode=RCIM_Cubic, leaveTangent=0.0224, tangentMode=RCTM_User, time=1, value=1}
         ]
        }}})))
   &ScaleColor(
    Linear Color Curve=DataInterface() {propertyValues {
      AlphaCurve {keys [<at 0 0>,<at 0.0716 0.2487>,<at 0.1723 1>,<at 0.504 0.6985>,<at 0.7379 0.2792>,<at 1 0>]}
      BlueCurve {keys [<at 0 1>,<at 1 1>]}
      GreenCurve {keys [<at 0 1>,<at 1 1>]}
      RedCurve {keys [<at 0 1>,<at 1 1>]}
     }}
    Scale Mode=2
   ) {module=ScaleColor001}
   &InheritSourceMovement(Applied Position Scale=<0.7 0.7 1>, ApplyForce=false, ApplyPosition=true, ApplyPositionFalloff=false)
   &SolveForcesAndVelocity(Clamp Velocity=false, Limit Acceleration=false)
   /SetVariables() {isSetParameters=true, module=SetVariables_#7}
   /SetVariables(
    Particles.ColorWave:LinearColor=&ColorFromCurve(ColorCurve=DataInterface() {propertyValues {
       AlphaCurve {keys [<at 0 0>,<at 0.1458 1.3>,<at 1 0 RCIM_Cubic>]}
       BlueCurve {keys [<at 0 0.1633 RCIM_Cubic>,<at 0.6071 0.0099>]}
       GreenCurve {keys [<at 0 0.1 RCIM_Cubic>,<at 0.6071 0.0061>]}
       RedCurve {keys [<at 0 2 RCIM_Cubic>,<at 0.6071 0.1217>]}
      }})
    Particles.WaveScale:Vector2f=&Vector2DFromFloat(Value=&FloatFromCurve(
      FloatCurve=DataInterface() {propertyValues {Curve {
         defaultValue=1
         keys [
          <at 0 0 RCIM_Cubic>
          {arriveTangent=2.3128, interpMode=RCIM_Cubic, leaveTangent=2.3128, time=0.018, value=0.5912}
          {arriveTangent=0.4163, interpMode=RCIM_Cubic, leaveTangent=0.4163, time=0.393, value=0.9088}
          <at 1 1 RCIM_Cubic>
         ]
        }}}
      Scale Curve=62.4812
     ))
   ) {isSetParameters=true, module=SetVariables_#10}
   /SetVariables(
    Particles.ColorRays:LinearColor=&ColorFromCurve(ColorCurve=DataInterface() {propertyValues {
       AlphaCurve {keys [<at 0 0>,<at 0.1458 0.14>,<at 0.9944 0>]}
       BlueCurve {keys [{arriveTangent=0.0166, interpMode=RCIM_Cubic, time=0.1661, value=0.0667}]}
       GreenCurve {keys [{arriveTangent=0.0101, interpMode=RCIM_Cubic, time=0.1661, value=0}]}
       RedCurve {keys [{arriveTangent=0.2029, interpMode=RCIM_Cubic, time=0.1661, value=2}]}
      }})
    Particles.RaysScale:Vector2f=&Vector2DFromFloat(Value=&FloatFromCurve(
      FloatCurve=DataInterface() {propertyValues {Curve {
         defaultValue=1
         keys [
          <at 0 0 RCIM_Cubic>
          {arriveTangent=0.4651, interpMode=RCIM_Cubic, leaveTangent=0.4651, tangentMode=RCTM_User, time=0.0146, value=0.6971}
          {arriveTangent=0.3074, interpMode=RCIM_Cubic, leaveTangent=0.3074, time=0.393, value=0.9088}
          <at 1 1 RCIM_Cubic>
         ]
        }}}
      Scale Curve=94.1286
     ))
   ) {isSetParameters=true, module=SetVariables_#11}
   /SetVariables(
    Particles.ColorFlash:LinearColor=&ColorFromCurve(ColorCurve=DataInterface() {propertyValues {
       AlphaCurve {keys [<at 0 0>,<at 0.1559 1>,<at 0.391 0.1>,<at 0.6898 0>]}
       BlueCurve {keys [{arriveTangent=-3.9747, interpMode=RCIM_Cubic, time=0.0355, value=0.2527}, <at 0.4374 0.0783>]}
       GreenCurve {keys [{arriveTangent=-0.9116, interpMode=RCIM_Cubic, time=0.0355, value=0.1565}, <at 0.4374 0.0783>]}
       RedCurve {keys [{arriveTangent=-4.7107, interpMode=RCIM_Cubic, time=0.0355, value=2}, <at 0.4374 3>]}
      }})
    Particles.FlashScale:Vector2f=&Vector2DFromCurve(
     Vector2Curve=DataInterface() {propertyValues {
       XCurve {keys [{arriveTangent=1.6105, interpMode=RCIM_Cubic, leaveTangent=1.6105, tangentMode=RCTM_User, time=0, value=0.3937}, <at 1 1 RCIM_Cubic>]}
       YCurve {keys [
         <at 0 0.2877 RCIM_Cubic>
         {arriveTangent=1.3271, interpMode=RCIM_Cubic, leaveTangent=1.3271, time=0.0463, value=1}
         {arriveTangent=-0.5243, interpMode=RCIM_Cubic, leaveTangent=-0.5243, time=0.16, value=0.5}
         <at 1 0.5 RCIM_Cubic>
        ]}
      }}
     ScaleCurve=<70 80>
    )
   ) {isSetParameters=true, module=SetVariables_#12}
   /SetVariables(
    Particles.ColorNebula:LinearColor=&MultiplyLinearColor(Linear Color=$Particles.Color, Scale Factor=<0.7 0.7 0.7 0.3>)
    Particles.NebulaScale:Vector2f=<80 80>
   ) {isSetParameters=true, module=SetVariables_#13}
   /SetVariables(
    Particles.ExtraScale:Vector3f=&Multiply_Vector(A=$Particles.Scale, B=<0.7 0.7 1>)
    Particles.MeshOrientation2:Quat4f=&OrientMeshToVector(
     Look At Direction=&RandomRangeVector(Minimum=<-1 -1 0>, Maximum=<1 1 0>)
     Reference Axis Vector=<1 0 0>
    )
    Particles.WindColor:LinearColor=&ColorFromCurve(ColorCurve=DataInterface() {propertyValues {
       AlphaCurve {keys [
         <at 0 0.664>
         {arriveTangent=-0.8348, interpMode=RCIM_Cubic, leaveTangent=-0.8348, time=0.0602, value=1.2}
         <at 0.758 0.0313>
         <at 0.9451 0 RCIM_Cubic>
        ]}
       BlueCurve {keys [<at 0 0.4696>,<at 0.4954 0.1696 RCIM_Cubic>]}
       GreenCurve {keys [<at 0 0.4696>,<at 0.4954 0.1696 RCIM_Cubic>]}
       RedCurve {keys [<at 0 2>,<at 0.4954 1.5 RCIM_Cubic>]}
      }})
   ) {isSetParameters=true, module=SetVariables_#14}
   /SetVariables(
    Particles.ExtraScale:Vector3f=&Multiply_Vector(A=$Particles.Scale, B=<0.7 0.7 1>)
    Particles.MeshOrientation2:Quat4f=&OrientMeshToVector(
     Look At Direction=&RandomRangeVector(Minimum=<-1 -1 0>, Maximum=<1 1 0>)
     Reference Axis Vector=<1 0 0>
    )
    Particles.WindColor:LinearColor=&ColorFromCurve(ColorCurve=DataInterface() {propertyValues {
       AlphaCurve {keys [
         <at 0 0.664>
         {arriveTangent=-0.8348, interpMode=RCIM_Cubic, leaveTangent=-0.8348, time=0.0602, value=1.2}
         <at 0.758 0.0313>
         <at 0.9451 0 RCIM_Cubic>
        ]}
       BlueCurve {keys [<at 0 1.0957>,<at 0.4954 0.1696 RCIM_Cubic>]}
       GreenCurve {keys [<at 0 1.0957>,<at 0.4954 0.1696 RCIM_Cubic>]}
       RedCurve {keys [<at 0 3>,<at 0.4954 1.5 RCIM_Cubic>]}
      }})
   ) {enabled=false, isSetParameters=true, module=SetVariables_#15}
  ]
  rendererClasses [&NiagaraSpriteRendererProperties, &NiagaraMeshRendererProperties]
  renderers [
   {rendererClass=&NiagaraSpriteRendererProperties}
   {rendererClass=&NiagaraSpriteRendererProperties, rendererIndex=1}
   {rendererClass=&NiagaraSpriteRendererProperties, rendererIndex=2}
   {rendererClass=&NiagaraSpriteRendererProperties, rendererIndex=3}
   {rendererClass=&NiagaraMeshRendererProperties, rendererIndex=4}
   {rendererClass=&NiagaraMeshRendererProperties, rendererIndex=5}
  ]
 }
 Flares_Reg {
  emitterUpdateScript [
   &EmitterState(
    Delay First Loop Only=true
    Life Cycle Mode=::NewEnumerator1
    Loop Delay=&Divide_Float(A=70, B=30)
    Loop Duration=&Divide_Float(A=20, B=30)
    UseLoopDelay=true
   )
   &SpawnBurst_Instantaneous(
    Spawn Probability=$System.SpawnRegular
    Spawn Time=&Divide_Float(A=&MakeCustomFloatFromBool(Bool=$System.bSpawnBack, True Float=9, False Float=1), B=30)
    Use Spawn Probability=true
    UseLoopCountLimit=false
   )
   &SpawnBurst_Instantaneous(
    Spawn Group=1
    Spawn Probability=$System.SpawnRegular
    Spawn Time=&Divide_Float(A=&MakeCustomFloatFromBool(Bool=$System.bSpawnBack, True Float=19, False Float=17), B=30)
    Use Spawn Probability=true
    UseLoopCountLimit=false
   ) {module=SpawnBurst_Instantaneous001}
   &SpawnBurst_Instantaneous(
    Spawn Group=2
    Spawn Probability=$System.SpawnRegular
    Spawn Time=&Divide_Float(A=&MakeCustomFloatFromBool(Bool=$System.bSpawnBack, True Float=14, False Float=13), B=30)
    Use Spawn Probability=true
    UseLoopCountLimit=false
   ) {module=SpawnBurst_Instantaneous002}
   &SpawnBurst_Instantaneous(
    Spawn Group=3
    Spawn Probability=$System.SpawnRegular
    Spawn Time=&Divide_Float(A=&MakeCustomFloatFromBool(Bool=$System.bSpawnBack, True Float=4, False Float=9), B=30)
    Use Spawn Probability=true
    UseLoopCountLimit=false
   ) {module=SpawnBurst_Instantaneous003}
  ]
  particleSpawnScript [
   &PartitionParticles(ID to Partition:NiagaraInt32=&ReturnExecIndex()) {enabled=false}
   &InitializeParticle(
    Lifetime=0.8
    Mesh Scale Mode=::NewEnumerator3
    Mesh Uniform Scale=0.4
    Position Mode=::NewEnumerator0
    Sprite Rotation Mode=::NewEnumerator2
    Sprite Size Mode=::NewEnumerator3
    Uniform Sprite Size=100
    Write Material Random=true
   )
   &SkeletalMeshLocation(
    Bone / Socket Index=&SpawnGroup()
    MaskSpawnGroup=false
    Mesh Sampling Type=::NewEnumerator5
    Orientation Sampling=::NewEnumerator1
    Skeletal Mesh=$System.Skeletal Mesh
    Socket Sampling Mode=::NewEnumerator5
   )
   &InitialMeshOrientation(
    Mesh Orientation Mode=::NewEnumerator2
    Orientation Coordinate Space=::World
    Rotation=&RandomRangeVector(Minimum=<0 0 0>, Maximum=<0 0 1>)
    Rotation Coordinate Space=::NewEnumerator1
    Use Rotation Vector=true
   )
   &SpriteFacingAndAlignment(
    Alignment Coordinate Space=::World
    Facing Coordinate Space=::World
    Sprite Alignment=&RandomVector(Vector Scale=1)
    Sprite Facing=<0 0 1>
    WriteAlignment=true
    WriteFacing=true
   )
   &AddVelocity(Velocity=<0 -50 0>, Velocity Speed Scale=&FN_FrontEndFloatModifier(Gameplay=0, Front End=1))
  ]
  particleUpdateScript [
   &ParticleState()
   &ScaleMeshSize(Scale Factor=&VectorFromFloat(Value=&FloatFromCurve(FloatCurve=DataInterface() {propertyValues {Curve {
         defaultValue=1
         keys [
          {arriveTangent=1.1277, interpMode=RCIM_Cubic, leaveTangent=1.1277, tangentMode=RCTM_User, time=0, value=0.2}
          {arriveTangent=0.8268, interpMode=RCIM_Cubic, leaveTangent=0.8268, tangentMode=RCTM_User, time=0.4507, value=0.6528}
          {arriveTangent=0.0224, interpMode=RCIM_Cubic, leaveTangent=0.0224, tangentMode=RCTM_User, time=1, value=1}
         ]
        }}})))
   &ScaleColor(
    Linear Color Curve=DataInterface() {propertyValues {
      AlphaCurve {keys [<at 0 0>,<at 0.0716 0.2487>,<at 0.1723 1>,<at 0.504 0.6985>,<at 0.7379 0.2792>,<at 1 0>]}
      BlueCurve {keys [<at 0 1>,<at 1 1>]}
      GreenCurve {keys [<at 0 1>,<at 1 1>]}
      RedCurve {keys [<at 0 1>,<at 1 1>]}
     }}
    Scale Mode=2
   ) {module=ScaleColor001}
   &InheritSourceMovement(Applied Position Scale=<0.7 0.7 1>, ApplyForce=false, ApplyPosition=true, ApplyPositionFalloff=false)
   &SolveForcesAndVelocity(Clamp Velocity=false, Limit Acceleration=false)
   /SetVariables() {isSetParameters=true, module=SetVariables_#7}
   /SetVariables(
    Particles.ColorWave:LinearColor=&ColorFromCurve(ColorCurve=DataInterface() {propertyValues {
       AlphaCurve {keys [<at 0 0>,<at 0.1458 1.3>,<at 1 0 RCIM_Cubic>]}
       BlueCurve {keys [<at 0 0.1633 RCIM_Cubic>,<at 0.6071 0.0099>]}
       GreenCurve {keys [<at 0 0.1 RCIM_Cubic>,<at 0.6071 0.0061>]}
       RedCurve {keys [<at 0 2 RCIM_Cubic>,<at 0.6071 0.1217>]}
      }})
    Particles.WaveScale:Vector2f=&Vector2DFromFloat(Value=&FloatFromCurve(
      FloatCurve=DataInterface() {propertyValues {Curve {
         defaultValue=1
         keys [
          <at 0 0 RCIM_Cubic>
          {arriveTangent=2.3128, interpMode=RCIM_Cubic, leaveTangent=2.3128, time=0.018, value=0.5912}
          {arriveTangent=0.4163, interpMode=RCIM_Cubic, leaveTangent=0.4163, time=0.393, value=0.9088}
          <at 1 1 RCIM_Cubic>
         ]
        }}}
      Scale Curve=62.4812
     ))
   ) {isSetParameters=true, module=SetVariables_#10}
   /SetVariables(
    Particles.ColorRays:LinearColor=&ColorFromCurve(ColorCurve=DataInterface() {propertyValues {
       AlphaCurve {keys [<at 0 0>,<at 0.1458 0.35>,<at 0.9944 0>]}
       BlueCurve {keys [{arriveTangent=0.0166, interpMode=RCIM_Cubic, time=0.1661, value=0.0667}]}
       GreenCurve {keys [{arriveTangent=0.0101, interpMode=RCIM_Cubic, time=0.1661, value=0}]}
       RedCurve {keys [{arriveTangent=0.2029, interpMode=RCIM_Cubic, time=0.1661, value=2}]}
      }})
    Particles.RaysScale:Vector2f=&Vector2DFromFloat(Value=&FloatFromCurve(
      FloatCurve=DataInterface() {propertyValues {Curve {
         defaultValue=1
         keys [
          <at 0 0 RCIM_Cubic>
          {arriveTangent=0.4651, interpMode=RCIM_Cubic, leaveTangent=0.4651, tangentMode=RCTM_User, time=0.0146, value=0.6971}
          {arriveTangent=0.3074, interpMode=RCIM_Cubic, leaveTangent=0.3074, time=0.393, value=0.9088}
          <at 1 1 RCIM_Cubic>
         ]
        }}}
      Scale Curve=94.1286
     ))
   ) {isSetParameters=true, module=SetVariables_#11}
   /SetVariables(
    Particles.ColorFlash:LinearColor=&ColorFromCurve(ColorCurve=DataInterface() {propertyValues {
       AlphaCurve {keys [<at 0 0>,<at 0.1559 1>,<at 0.391 0.1>,<at 0.6898 0>]}
       BlueCurve {keys [{arriveTangent=-3.9747, interpMode=RCIM_Cubic, time=0.0355, value=0.2527}, <at 0.4374 0.0783>]}
       GreenCurve {keys [{arriveTangent=-0.9116, interpMode=RCIM_Cubic, time=0.0355, value=0.1565}, <at 0.4374 0.0783>]}
       RedCurve {keys [{arriveTangent=-4.7107, interpMode=RCIM_Cubic, time=0.0355, value=2}, <at 0.4374 3>]}
      }})
    Particles.FlashScale:Vector2f=&Vector2DFromCurve(
     Vector2Curve=DataInterface() {propertyValues {
       XCurve {keys [{arriveTangent=1.6105, interpMode=RCIM_Cubic, leaveTangent=1.6105, tangentMode=RCTM_User, time=0, value=0.3937}, <at 1 1 RCIM_Cubic>]}
       YCurve {keys [
         <at 0 0.2877 RCIM_Cubic>
         {arriveTangent=1.3271, interpMode=RCIM_Cubic, leaveTangent=1.3271, time=0.0463, value=1}
         {arriveTangent=-0.5243, interpMode=RCIM_Cubic, leaveTangent=-0.5243, time=0.16, value=0.5}
         <at 1 0.5 RCIM_Cubic>
        ]}
      }}
     ScaleCurve=<70 80>
    )
   ) {isSetParameters=true, module=SetVariables_#12}
   /SetVariables(
    Particles.ColorNebula:LinearColor=&MultiplyLinearColor(Linear Color=$Particles.Color, Scale Factor=<0.7 0.7 0.7 0.45>)
    Particles.NebulaScale:Vector2f=<80 80>
   ) {isSetParameters=true, module=SetVariables_#13}
   /SetVariables(
    Particles.ExtraScale:Vector3f=&Multiply_Vector(A=$Particles.Scale, B=<0.7 0.7 1>)
    Particles.MeshOrientation2:Quat4f=&OrientMeshToVector(
     Look At Direction=&RandomRangeVector(Minimum=<-1 -1 0>, Maximum=<1 1 0>)
     Reference Axis Vector=<1 0 0>
    )
    Particles.WindColor:LinearColor=&ColorFromCurve(ColorCurve=DataInterface() {propertyValues {
       AlphaCurve {keys [
         <at 0 0.664>
         {arriveTangent=-0.8348, interpMode=RCIM_Cubic, leaveTangent=-0.8348, time=0.0602, value=1.2}
         <at 0.758 0.0313>
         <at 0.9451 0 RCIM_Cubic>
        ]}
       BlueCurve {keys [<at 0 0.4696>,<at 0.4954 0.1696 RCIM_Cubic>]}
       GreenCurve {keys [<at 0 0.4696>,<at 0.4954 0.1696 RCIM_Cubic>]}
       RedCurve {keys [<at 0 2>,<at 0.4954 1.5 RCIM_Cubic>]}
      }})
   ) {isSetParameters=true, module=SetVariables_#14}
   /SetVariables(
    Particles.ExtraScale:Vector3f=&Multiply_Vector(A=$Particles.Scale, B=<0.7 0.7 1>)
    Particles.MeshOrientation2:Quat4f=&OrientMeshToVector(
     Look At Direction=&RandomRangeVector(Minimum=<-1 -1 0>, Maximum=<1 1 0>)
     Reference Axis Vector=<1 0 0>
    )
    Particles.WindColor:LinearColor=&ColorFromCurve(ColorCurve=DataInterface() {propertyValues {
       AlphaCurve {keys [
         <at 0 0.664>
         {arriveTangent=-0.8348, interpMode=RCIM_Cubic, leaveTangent=-0.8348, time=0.0602, value=1.2}
         <at 0.758 0.0313>
         <at 0.9451 0 RCIM_Cubic>
        ]}
       BlueCurve {keys [<at 0 1.0957>,<at 0.4954 0.1696 RCIM_Cubic>]}
       GreenCurve {keys [<at 0 1.0957>,<at 0.4954 0.1696 RCIM_Cubic>]}
       RedCurve {keys [<at 0 3>,<at 0.4954 1.5 RCIM_Cubic>]}
      }})
   ) {enabled=false, isSetParameters=true, module=SetVariables_#15}
  ]
  rendererClasses [&NiagaraSpriteRendererProperties, &NiagaraMeshRendererProperties]
  renderers [
   {rendererClass=&NiagaraSpriteRendererProperties}
   {rendererClass=&NiagaraSpriteRendererProperties, rendererIndex=1}
   {rendererClass=&NiagaraSpriteRendererProperties, rendererIndex=2}
   {rendererClass=&NiagaraSpriteRendererProperties, rendererIndex=3}
   {rendererClass=&NiagaraMeshRendererProperties, rendererIndex=4}
   {rendererClass=&NiagaraMeshRendererProperties, rendererIndex=5}
  ]
 }
 Flares_Slow {
  emitterUpdateScript [
   &EmitterState(
    Delay First Loop Only=true
    Life Cycle Mode=::NewEnumerator1
    Loop Delay=&Divide_Float(A=70, B=30)
    Loop Duration=&Divide_Float(A=40, B=90)
    UseLoopDelay=true
   )
   &SpawnBurst_Instantaneous(
    Spawn Probability=&ClampFloat(Float=&Subtract_Float(A=$System.SpawnForward, B=$System.SpawnRegular))
    Spawn Time=&Divide_Float(A=2, B=90)
    Use Spawn Probability=true
    UseLoopCountLimit=false
   )
   &SpawnBurst_Instantaneous(
    Spawn Group=1
    Spawn Probability=&ClampFloat(Float=&Subtract_Float(A=$System.SpawnForward, B=$System.SpawnRegular))
    Spawn Time=&Divide_Float(A=34, B=90)
    Use Spawn Probability=true
    UseLoopCountLimit=false
   ) {module=SpawnBurst_Instantaneous001}
   &SpawnBurst_Instantaneous(
    Spawn Group=2
    Spawn Probability=&ClampFloat(Float=&Subtract_Float(A=$System.SpawnForward, B=$System.SpawnRegular))
    Spawn Time=&Divide_Float(A=26, B=90)
    Use Spawn Probability=true
    UseLoopCountLimit=false
   ) {module=SpawnBurst_Instantaneous002}
   &SpawnBurst_Instantaneous(
    Spawn Group=3
    Spawn Probability=&ClampFloat(Float=&Subtract_Float(A=$System.SpawnForward, B=$System.SpawnRegular))
    Spawn Time=&Divide_Float(A=18, B=90)
    Use Spawn Probability=true
    UseLoopCountLimit=false
   ) {module=SpawnBurst_Instantaneous003}
  ]
  particleSpawnScript [
   &PartitionParticles(ID to Partition:NiagaraInt32=&ReturnExecIndex()) {enabled=false}
   &InitializeParticle(
    Lifetime=0.8
    Mesh Scale Mode=::NewEnumerator3
    Mesh Uniform Scale=0.4
    Position Mode=::NewEnumerator0
    Sprite Rotation Mode=::NewEnumerator2
    Sprite Size Mode=::NewEnumerator3
    Uniform Sprite Size=100
    Write Material Random=true
   )
   &SkeletalMeshLocation(
    Bone / Socket Index=&SpawnGroup()
    MaskSpawnGroup=false
    Mesh Sampling Type=::NewEnumerator5
    Orientation Sampling=::NewEnumerator1
    Skeletal Mesh=$System.Skeletal Mesh
    Socket Sampling Mode=::NewEnumerator5
   )
   &InitialMeshOrientation(
    Mesh Orientation Mode=::NewEnumerator2
    Orientation Coordinate Space=::World
    Rotation=&RandomRangeVector(Minimum=<0 0 0>, Maximum=<0 0 1>)
    Rotation Coordinate Space=::NewEnumerator1
    Use Rotation Vector=true
   )
   &SpriteFacingAndAlignment(
    Alignment Coordinate Space=::World
    Facing Coordinate Space=::World
    Sprite Alignment=&RandomVector(Vector Scale=1)
    Sprite Facing=<0 0 1>
    WriteAlignment=true
    WriteFacing=true
   )
   &AddVelocity(Velocity=<0 -50 0>, Velocity Speed Scale=&FN_FrontEndFloatModifier(Gameplay=0, Front End=1))
  ]
  particleUpdateScript [
   &ParticleState()
   &ScaleMeshSize(Scale Factor=&VectorFromFloat(Value=&FloatFromCurve(FloatCurve=DataInterface() {propertyValues {Curve {
         defaultValue=1
         keys [
          {arriveTangent=1.1277, interpMode=RCIM_Cubic, leaveTangent=1.1277, tangentMode=RCTM_User, time=0, value=0.2}
          {arriveTangent=0.8268, interpMode=RCIM_Cubic, leaveTangent=0.8268, tangentMode=RCTM_User, time=0.4507, value=0.6528}
          {arriveTangent=0.0224, interpMode=RCIM_Cubic, leaveTangent=0.0224, tangentMode=RCTM_User, time=1, value=1}
         ]
        }}})))
   &ScaleColor(
    Linear Color Curve=DataInterface() {propertyValues {
      AlphaCurve {keys [<at 0 0>,<at 0.0716 0.2487>,<at 0.1723 1>,<at 0.504 0.6985>,<at 0.7379 0.2792>,<at 1 0>]}
      BlueCurve {keys [<at 0 1>,<at 1 1>]}
      GreenCurve {keys [<at 0 1>,<at 1 1>]}
      RedCurve {keys [<at 0 1>,<at 1 1>]}
     }}
    Scale Mode=2
   ) {module=ScaleColor001}
   &InheritSourceMovement(Applied Position Scale=<0.7 0.7 1>, ApplyForce=false, ApplyPosition=true, ApplyPositionFalloff=false)
   &SolveForcesAndVelocity(Clamp Velocity=false, Limit Acceleration=false)
   /SetVariables() {isSetParameters=true, module=SetVariables_#7}
   /SetVariables(
    Particles.ColorWave:LinearColor=&ColorFromCurve(ColorCurve=DataInterface() {propertyValues {
       AlphaCurve {keys [<at 0 0>,<at 0.1458 1.3>,<at 1 0 RCIM_Cubic>]}
       BlueCurve {keys [<at 0 0.1633 RCIM_Cubic>,<at 0.6071 0.0099>]}
       GreenCurve {keys [<at 0 0.1 RCIM_Cubic>,<at 0.6071 0.0061>]}
       RedCurve {keys [<at 0 2 RCIM_Cubic>,<at 0.6071 0.1217>]}
      }})
    Particles.WaveScale:Vector2f=&Vector2DFromFloat(Value=&FloatFromCurve(
      FloatCurve=DataInterface() {propertyValues {Curve {
         defaultValue=1
         keys [
          <at 0 0 RCIM_Cubic>
          {arriveTangent=2.3128, interpMode=RCIM_Cubic, leaveTangent=2.3128, time=0.018, value=0.5912}
          {arriveTangent=0.4163, interpMode=RCIM_Cubic, leaveTangent=0.4163, time=0.393, value=0.9088}
          <at 1 1 RCIM_Cubic>
         ]
        }}}
      Scale Curve=62.4812
     ))
   ) {isSetParameters=true, module=SetVariables_#10}
   /SetVariables(
    Particles.ColorRays:LinearColor=&ColorFromCurve(ColorCurve=DataInterface() {propertyValues {
       AlphaCurve {keys [<at 0 0>,<at 0.1458 0.35>,<at 0.9944 0>]}
       BlueCurve {keys [{arriveTangent=0.0166, interpMode=RCIM_Cubic, time=0.1661, value=0.0667}]}
       GreenCurve {keys [{arriveTangent=0.0101, interpMode=RCIM_Cubic, time=0.1661, value=0}]}
       RedCurve {keys [{arriveTangent=0.2029, interpMode=RCIM_Cubic, time=0.1661, value=2}]}
      }})
    Particles.RaysScale:Vector2f=&Vector2DFromFloat(Value=&FloatFromCurve(
      FloatCurve=DataInterface() {propertyValues {Curve {
         defaultValue=1
         keys [
          <at 0 0 RCIM_Cubic>
          {arriveTangent=0.4651, interpMode=RCIM_Cubic, leaveTangent=0.4651, tangentMode=RCTM_User, time=0.0146, value=0.6971}
          {arriveTangent=0.3074, interpMode=RCIM_Cubic, leaveTangent=0.3074, time=0.393, value=0.9088}
          <at 1 1 RCIM_Cubic>
         ]
        }}}
      Scale Curve=94.1286
     ))
   ) {isSetParameters=true, module=SetVariables_#11}
   /SetVariables(
    Particles.ColorFlash:LinearColor=&ColorFromCurve(ColorCurve=DataInterface() {propertyValues {
       AlphaCurve {keys [<at 0 0>,<at 0.1559 1>,<at 0.391 0.1>,<at 0.6898 0>]}
       BlueCurve {keys [{arriveTangent=-3.9747, interpMode=RCIM_Cubic, time=0.0355, value=0.2527}, <at 0.4374 0.0783>]}
       GreenCurve {keys [{arriveTangent=-0.9116, interpMode=RCIM_Cubic, time=0.0355, value=0.1565}, <at 0.4374 0.0783>]}
       RedCurve {keys [{arriveTangent=-4.7107, interpMode=RCIM_Cubic, time=0.0355, value=2}, <at 0.4374 3>]}
      }})
    Particles.FlashScale:Vector2f=&Vector2DFromCurve(
     Vector2Curve=DataInterface() {propertyValues {
       XCurve {keys [{arriveTangent=1.6105, interpMode=RCIM_Cubic, leaveTangent=1.6105, tangentMode=RCTM_User, time=0, value=0.3937}, <at 1 1 RCIM_Cubic>]}
       YCurve {keys [
         <at 0 0.2877 RCIM_Cubic>
         {arriveTangent=1.3271, interpMode=RCIM_Cubic, leaveTangent=1.3271, time=0.0463, value=1}
         {arriveTangent=-0.5243, interpMode=RCIM_Cubic, leaveTangent=-0.5243, time=0.16, value=0.5}
         <at 1 0.5 RCIM_Cubic>
        ]}
      }}
     ScaleCurve=<70 80>
    )
   ) {isSetParameters=true, module=SetVariables_#12}
   /SetVariables(
    Particles.ColorNebula:LinearColor=&MultiplyLinearColor(Linear Color=$Particles.Color, Scale Factor=<0.7 0.7 0.7 0.45>)
    Particles.NebulaScale:Vector2f=<80 80>
   ) {isSetParameters=true, module=SetVariables_#13}
   /SetVariables(
    Particles.ExtraScale:Vector3f=&Multiply_Vector(A=$Particles.Scale, B=<0.7 0.7 1>)
    Particles.MeshOrientation2:Quat4f=&OrientMeshToVector(
     Look At Direction=&RandomRangeVector(Minimum=<-1 -1 0>, Maximum=<1 1 0>)
     Reference Axis Vector=<1 0 0>
    )
    Particles.WindColor:LinearColor=&ColorFromCurve(ColorCurve=DataInterface() {propertyValues {
       AlphaCurve {keys [
         <at 0 0.664>
         {arriveTangent=-0.8348, interpMode=RCIM_Cubic, leaveTangent=-0.8348, time=0.0602, value=1.2}
         <at 0.758 0.0313>
         <at 0.9451 0 RCIM_Cubic>
        ]}
       BlueCurve {keys [<at 0 0.4696>,<at 0.4954 0.1696 RCIM_Cubic>]}
       GreenCurve {keys [<at 0 0.4696>,<at 0.4954 0.1696 RCIM_Cubic>]}
       RedCurve {keys [<at 0 2>,<at 0.4954 1.5 RCIM_Cubic>]}
      }})
   ) {isSetParameters=true, module=SetVariables_#14}
   /SetVariables(
    Particles.ExtraScale:Vector3f=&Multiply_Vector(A=$Particles.Scale, B=<0.7 0.7 1>)
    Particles.MeshOrientation2:Quat4f=&OrientMeshToVector(
     Look At Direction=&RandomRangeVector(Minimum=<-1 -1 0>, Maximum=<1 1 0>)
     Reference Axis Vector=<1 0 0>
    )
    Particles.WindColor:LinearColor=&ColorFromCurve(ColorCurve=DataInterface() {propertyValues {
       AlphaCurve {keys [
         <at 0 0.664>
         {arriveTangent=-0.8348, interpMode=RCIM_Cubic, leaveTangent=-0.8348, time=0.0602, value=1.2}
         <at 0.758 0.0313>
         <at 0.9451 0 RCIM_Cubic>
        ]}
       BlueCurve {keys [<at 0 1.0957>,<at 0.4954 0.1696 RCIM_Cubic>]}
       GreenCurve {keys [<at 0 1.0957>,<at 0.4954 0.1696 RCIM_Cubic>]}
       RedCurve {keys [<at 0 3>,<at 0.4954 1.5 RCIM_Cubic>]}
      }})
   ) {enabled=false, isSetParameters=true, module=SetVariables_#15}
  ]
  rendererClasses [&NiagaraSpriteRendererProperties, &NiagaraMeshRendererProperties]
  renderers [
   {rendererClass=&NiagaraSpriteRendererProperties}
   {rendererClass=&NiagaraSpriteRendererProperties, rendererIndex=1}
   {rendererClass=&NiagaraSpriteRendererProperties, rendererIndex=2}
   {rendererClass=&NiagaraSpriteRendererProperties, rendererIndex=3}
   {rendererClass=&NiagaraMeshRendererProperties, rendererIndex=4}
   {rendererClass=&NiagaraMeshRendererProperties, rendererIndex=5}
  ]
 }
 Sparks_Fwd001 {
  bEnabled=false
  emitterUpdateScript [
   &EmitterState(
    Delay First Loop Only=true
    Life Cycle Mode=::NewEnumerator1
    Loop Delay=&Divide_Float(A=70, B=30)
    Loop Duration=&Divide_Float(A=20, B=30)
    UseLoopDelay=true
   )
   &SpawnBurst_Instantaneous(
    Spawn Count=&RandomRangeInt(Minimum=5, Maximum=8)
    Spawn Probability=&ClampFloat(Float=&Subtract_Float(A=$System.SpawnForward, B=$System.SpawnRegular))
    Spawn Time=&Divide_Float(A=2, B=90)
    Use Spawn Probability=true
    UseLoopCountLimit=false
   )
   &SpawnBurst_Instantaneous(
    Spawn Count=&RandomRangeInt(Minimum=5, Maximum=8)
    Spawn Group=1
    Spawn Probability=&ClampFloat(Float=&Subtract_Float(A=$System.SpawnForward, B=$System.SpawnRegular))
    Spawn Time=&Divide_Float(A=34, B=90)
    Use Spawn Probability=true
    UseLoopCountLimit=false
   ) {module=SpawnBurst_Instantaneous001}
   &SpawnBurst_Instantaneous(
    Spawn Count=&RandomRangeInt(Minimum=5, Maximum=8)
    Spawn Group=2
    Spawn Probability=&ClampFloat(Float=&Subtract_Float(A=$System.SpawnForward, B=$System.SpawnRegular))
    Spawn Time=&Divide_Float(A=26, B=90)
    Use Spawn Probability=true
    UseLoopCountLimit=false
   ) {module=SpawnBurst_Instantaneous002}
   &SpawnBurst_Instantaneous(
    Spawn Count=&RandomRangeInt(Minimum=5, Maximum=8)
    Spawn Group=3
    Spawn Probability=&ClampFloat(Float=&Subtract_Float(A=$System.SpawnForward, B=$System.SpawnRegular))
    Spawn Time=&Divide_Float(A=18, B=90)
    Use Spawn Probability=true
    UseLoopCountLimit=false
   ) {module=SpawnBurst_Instantaneous003}
  ]
  particleSpawnScript [
   &PartitionParticles(
    A:NiagaraFloat=&RandomRangeFloat(Minimum=0, Maximum=0.8)
    B=0.5
    Partition Mode:ENiagara_PartitionMode_v2=::NewEnumerator4
   )
   /SetVariables(Particles.VisibilityTag=$Particles.PartitionParticles.Partition) {isSetParameters=true, module=SetVariables_#16}
   &InitializeParticle(
    Color=&RandomRangeLinearColor(Minimum=<rgba 4 0.2435 0.2435 0.2435>, Maximum=<rgba 5 0.3478 0.7355 0.713>)
    Lifetime=0.8
    Mesh Scale=<0.4 0.4 0.5>
    Mesh Scale Mode=::NewEnumerator4
    Position Mode=::NewEnumerator0
    Sprite Size Mode=::NewEnumerator2
    Write Material Random=true
   )
   &SkeletalMeshLocation(
    Bone / Socket Index=&SpawnGroup()
    MaskSpawnGroup=false
    Mesh Sampling Type=::NewEnumerator5
    Orientation Sampling=::NewEnumerator1
    Skeletal Mesh=$System.Skeletal Mesh
    Socket Sampling Mode=::NewEnumerator5
   )
   &ShapeLocation(
    MaskSpawnGroup=false
    Override Seed=false
    Ring Radius=13
    Rotation Coordinate Space=::World
    Shape Primitive=::NewEnumerator4
    UseOwnerScale=true
   )
   &AddVelocity(
    Rotation Coordinate Space=::World
    Velocity=&AddVectorToPosition(
     Position=&RandomVector(Vector Scale=0.5)
     Vector=&RandomRangeVector(Minimum=<0 0 0.6>, Maximum=<0 0 0.8>)
    )
    Velocity Speed Scale=&MaskFloat(Mask Source=1, Mask Parameter=$Particles.PartitionParticles.Partition, Mask=0, Masked Float=70, Passthrough Float=40)
   ) {module=AddVelocity001}
   &AddVelocity(Velocity=<0 -50 0>, Velocity Speed Scale=&FN_FrontEndFloatModifier(Gameplay=0, Front End=1))
   &InitialMeshOrientation(
    Mesh Axis To Orient=<0 0 1>
    Mesh Orientation Mode=::NewEnumerator2
    Orientation Vector=<0 0 1>
    Rotation=&RandomRangeVector(Minimum=<0 0 0>, Maximum=<0 0 1>)
    Use Rotation Vector=true
   )
  ]
  particleUpdateScript [
   &ParticleState()
   &Drag(
    Drag=&MaskFloat(Mask Source=1, Mask Parameter=$Particles.PartitionParticles.Partition, Mask=0, Masked Float=1.5, Passthrough Float=0.7)
    Rotational Drag=1.5
    Use Linear Drag=true
    Use Rotational Drag=true
   )
   &ScaleSpriteSize(
    Scale Sprite Size Mode=::NewEnumerator0
    Uniform Scale Factor=&Multiply_Float(
     A=&FloatFromCurve(FloatCurve=DataInterface() {propertyValues {Curve {
         defaultValue=1
         keys [
          {arriveTangent=6.4214, interpMode=RCIM_Cubic, leaveTangent=6.4214, tangentMode=RCTM_User, time=0, value=0}
          {interpMode=RCIM_Cubic, tangentMode=RCTM_User, time=0.2882, value=1}
          {arriveTangent=-2.7349, tangentMode=RCTM_User, time=1, value=0}
         ]
        }}})
     B=&FloatFromCurve(
      FloatCurve=DataInterface() {propertyValues {Curve {keys [<at 0 0.2>,<at 0.1 0.8647>,<at 0.7 1>,<at 0.85 1.3>]}}}
      CurveIndex=&RandomRangeFloat(Minimum=0, Maximum=1, Evaluation Type=::NewEnumerator1)
     )
    )
   )
   &ScaleColor(
    Linear Color Curve=DataInterface() {propertyValues {
      AlphaCurve {keys [<at 0 0>,<at 0.0716 0.2487>,<at 0.1723 1>,<at 0.504 0.6985>,<at 0.7379 0.2792>,<at 1 0>]}
      BlueCurve {keys [<at 0 1>,<at 1 1>]}
      GreenCurve {keys [<at 0 1>,<at 1 1>]}
      RedCurve {keys [<at 0 1>,<at 1 1>]}
     }}
    Scale Mode=2
   )
   &InheritSourceMovement(Applied Position Scale=<0.7 0.7 1>, ApplyForce=false, ApplyPosition=true, ApplyPositionFalloff=false)
   &SolveForcesAndVelocity(Clamp Velocity=false, Limit Acceleration=false)
   /SetVariables(
    Particles.Crystal Color:LinearColor=/Niagara/DynamicInputs/Add/Add_LinearColor(
     A:LinearColor=&ColorFromCurve(ColorCurve=DataInterface() {propertyValues {
        AlphaCurve {keys [<at 0.5781 1 RCIM_Cubic>,<at 1 0 RCIM_Cubic>]}
        BlueCurve {keys [<at 0 0 RCIM_Cubic>]}
        GreenCurve {keys [<at 0 0.0145 RCIM_Cubic>]}
        RedCurve {keys [<at 0 0.868 RCIM_Cubic>]}
       }})
     B:LinearColor=&MakeLinearColorFromVectorAndFloat(
      "Vector (RGB)"=&VectorFromFloat(Value=&FloatFromCurve(
        FloatCurve=DataInterface() {propertyValues {Curve {keys [<at 0.1365 0>,<at 0.1754 0.6533>,<at 0.2838 0>,<at 0.345 0.2634>,<at 0.4303 0.149>,<at 0.5361 0.4843>,<at 0.5674 0>,<at 1.3326 0>,<at 1.3532 0.2307>,<at 1.4591 0>,<at 1.5391 0.6635>,<at 1.6457 0>,<at 1.7194 0.4348>,<at 1.7831 0>,<at 2.316 0>,<at 2.36 0.5083>,<at 2.4843 0>,<at 3 0>]}}}
        CurveIndex=&Add_Float(A=$Particles.NormalizedAge, B=&RandomRangeFloat(Minimum=-0.2, Maximum=2))
        Scale Curve=&RandomRangeFloat(Minimum=0.2, Maximum=0.6)
       ))
      "Float (Alpha)"=0
     )
    )
    Particles.GlowAlignment:Vector3f=<0 0 1>
    Particles.GlowOffset=-5
    Particles.GlowSize=&Multiply_Vector2D(
     A=&MakeVector2D(
      X=&MakeFloatFromVector(Vector=$Particles.Scale, Channel=::NewEnumerator2)
      Y=&MakeFloatFromVector(Vector=$Particles.Scale, Channel=::NewEnumerator0)
     )
     B=<20.8 50.3>
    )
   ) {isSetParameters=true, module=SetVariables_#17}
   &ScaleMeshSize(Scale Factor=&VectorFromFloat(Value=&FloatFromCurve(FloatCurve=DataInterface() {propertyValues {Curve {
         defaultValue=1
         keys [
          {arriveTangent=5.1268, interpMode=RCIM_Cubic, leaveTangent=5.1268, tangentMode=RCTM_User, time=0, value=0}
          {
           arriveTangentWeight=0.2927
           interpMode=RCIM_Cubic
           leaveTangentWeight=0.2927
           tangentMode=RCTM_User
           tangentWeightMode=RCTWM_WeightedBoth
           time=0.5
           value=1
          }
          {arriveTangent=-3.873, tangentMode=RCTM_User, time=1, value=0}
         ]
        }}})))
   &UpdateMeshOrientation(Rotation Rate=&RandomRangeFloat(Minimum=-1, Maximum=1), Rotation Vector=<0 0 1>)
  ]
  rendererClasses [&NiagaraSpriteRendererProperties, &NiagaraMeshRendererProperties]
  renderers [
   {rendererClass=&NiagaraSpriteRendererProperties}
   {rendererClass=&NiagaraSpriteRendererProperties, rendererIndex=1}
   {rendererClass=&NiagaraMeshRendererProperties, rendererIndex=2}
   {rendererClass=&NiagaraSpriteRendererProperties, rendererIndex=3}
  ]
 }
 Sparks_Reg {
  emitterUpdateScript [
   &EmitterState(
    Delay First Loop Only=true
    Life Cycle Mode=::NewEnumerator1
    Loop Delay=&Divide_Float(A=70, B=30)
    Loop Duration=&Divide_Float(A=20, B=30)
    UseLoopDelay=true
   )
   &SpawnBurst_Instantaneous(
    Spawn Count=&RandomRangeInt(Minimum=5, Maximum=8)
    Spawn Probability=$System.SpawnRegular
    Spawn Time=&Divide_Float(A=&MakeCustomFloatFromBool(Bool=$System.bSpawnBack, True Float=9, False Float=1), B=30)
    Use Spawn Probability=true
    UseLoopCountLimit=false
   )
   &SpawnBurst_Instantaneous(
    Spawn Count=&RandomRangeInt(Minimum=5, Maximum=8)
    Spawn Group=1
    Spawn Probability=$System.SpawnRegular
    Spawn Time=&Divide_Float(A=&MakeCustomFloatFromBool(Bool=$System.bSpawnBack, True Float=19, False Float=17), B=30)
    Use Spawn Probability=true
    UseLoopCountLimit=false
   ) {module=SpawnBurst_Instantaneous001}
   &SpawnBurst_Instantaneous(
    Spawn Count=&RandomRangeInt(Minimum=5, Maximum=8)
    Spawn Group=2
    Spawn Probability=$System.SpawnRegular
    Spawn Time=&Divide_Float(A=&MakeCustomFloatFromBool(Bool=$System.bSpawnBack, True Float=14, False Float=13), B=30)
    Use Spawn Probability=true
    UseLoopCountLimit=false
   ) {module=SpawnBurst_Instantaneous002}
   &SpawnBurst_Instantaneous(
    Spawn Count=&RandomRangeInt(Minimum=5, Maximum=8)
    Spawn Group=3
    Spawn Probability=$System.SpawnRegular
    Spawn Time=&Divide_Float(A=&MakeCustomFloatFromBool(Bool=$System.bSpawnBack, True Float=4, False Float=9), B=30)
    Use Spawn Probability=true
    UseLoopCountLimit=false
   ) {module=SpawnBurst_Instantaneous003}
  ]
  particleSpawnScript [
   &PartitionParticles(
    A:NiagaraFloat=&RandomRangeFloat(Minimum=0, Maximum=0.6)
    B=0.5
    Partition Mode:ENiagara_PartitionMode_v2=::NewEnumerator4
   )
   /SetVariables(Particles.VisibilityTag=$Particles.PartitionParticles.Partition) {isSetParameters=true, module=SetVariables_#16}
   &InitializeParticle(
    Color=&RandomRangeLinearColor(Minimum=<rgba 4 0.2435 0.2435 0.2435>, Maximum=<rgba 5 0.3478 0.7355 0.713>)
    Lifetime=0.8
    Mesh Scale Max=<0.85 0.85 0.95>
    Mesh Scale Min=<0.7 0.7 0.8>
    Mesh Scale Mode=::NewEnumerator2
    Position Mode=::NewEnumerator0
    Sprite Size Max=<20 20>
    Sprite Size Min=<10 10>
    Sprite Size Mode=::NewEnumerator2
    Write Material Random=true
   )
   &SkeletalMeshLocation(
    Bone / Socket Index=&SpawnGroup()
    MaskSpawnGroup=false
    Mesh Sampling Type=::NewEnumerator5
    Orientation Sampling=::NewEnumerator1
    Skeletal Mesh=$System.Skeletal Mesh
    Socket Sampling Mode=::NewEnumerator5
   )
   &ShapeLocation(
    MaskSpawnGroup=false
    Override Seed=false
    Ring Radius=13
    Rotation Coordinate Space=::World
    Shape Primitive=::NewEnumerator4
    UseOwnerScale=true
   )
   &InitialMeshOrientation(
    Mesh Axis To Orient=<0 0 1>
    Mesh Orientation Mode=::NewEnumerator2
    Orientation Vector=<0 0 1>
    Rotation=&RandomRangeVector(Minimum=<0 0 0>, Maximum=<0 0 1>)
    Use Rotation Vector=true
   )
   &AddVelocity(
    Rotation Coordinate Space=::World
    Velocity=&AddVectorToPosition(
     Position=&RandomVector(Vector Scale=&MaskFloat(Mask Source=1, Mask Parameter=$Particles.PartitionParticles.Partition, Mask=0, Masked Float=0.5, Passthrough Float=1.6))
     Vector=&Multiply_VectorByFloat(
      Vector=<0 0 1>
      Float=&MaskFloat(
       Mask Source=1
       Mask Parameter=$Particles.PartitionParticles.Partition
       Mask=0
       Masked Float=&RandomRangeFloat(Minimum=0.6, Maximum=0.8)
       Passthrough Float=0
      )
     )
    )
    Velocity Speed Scale=70
   ) {module=AddVelocity001}
   &AddVelocity(Velocity=<0 -50 0>, Velocity Speed Scale=&FN_FrontEndFloatModifier(Gameplay=0, Front End=1))
  ]
  particleUpdateScript [
   &ParticleState()
   &Drag(
    Drag=&MaskFloat(Mask Source=1, Mask Parameter=$Particles.PartitionParticles.Partition, Mask=0, Masked Float=1.5, Passthrough Float=2)
    Rotational Drag=1.5
    Use Linear Drag=true
    Use Rotational Drag=true
   )
   &ScaleSpriteSize(
    Scale Sprite Size Mode=::NewEnumerator0
    Uniform Scale Factor=&Multiply_Float(
     A=&FloatFromCurve(FloatCurve=DataInterface() {propertyValues {Curve {
         defaultValue=1
         keys [
          {arriveTangent=6.4214, interpMode=RCIM_Cubic, leaveTangent=6.4214, tangentMode=RCTM_User, time=0, value=0}
          {interpMode=RCIM_Cubic, tangentMode=RCTM_User, time=0.2882, value=1}
          {arriveTangent=-2.7349, tangentMode=RCTM_User, time=1, value=0}
         ]
        }}})
     B=&FloatFromCurve(
      FloatCurve=DataInterface() {propertyValues {Curve {keys [<at 0 0.2>,<at 0.1 0.8647>,<at 0.7 1>,<at 0.85 1.3>]}}}
      CurveIndex=&RandomRangeFloat(Minimum=0, Maximum=1, Evaluation Type=::NewEnumerator1)
     )
    )
   )
   &ScaleColor(
    Linear Color Curve=DataInterface() {propertyValues {
      AlphaCurve {keys [<at 0 0>,<at 0.0716 0.2487>,<at 0.1723 1>,<at 0.504 0.6985>,<at 0.7379 0.2792>,<at 1 0>]}
      BlueCurve {keys [<at 0 1>,<at 1 1>]}
      GreenCurve {keys [<at 0 1>,<at 1 1>]}
      RedCurve {keys [<at 0 1>,<at 1 1>]}
     }}
    Scale Mode=2
   )
   &InheritSourceMovement(Applied Position Scale=<0.7 0.7 1>, ApplyForce=false, ApplyPosition=true, ApplyPositionFalloff=false)
   &SolveForcesAndVelocity(Clamp Velocity=false, Limit Acceleration=false)
   /SetVariables(
    Particles.Crystal Color:LinearColor=&Add_LinearColor(
     A:LinearColor=&ColorFromCurve(ColorCurve=DataInterface() {propertyValues {
        AlphaCurve {keys [<at 0.5781 1 RCIM_Cubic>,<at 1 0 RCIM_Cubic>]}
        BlueCurve {keys [<at 0 0 RCIM_Cubic>]}
        GreenCurve {keys [<at 0 0.0145 RCIM_Cubic>]}
        RedCurve {keys [<at 0 0.868 RCIM_Cubic>]}
       }})
     B:LinearColor=&MakeLinearColorFromVectorAndFloat(
      "Vector (RGB)"=&VectorFromFloat(Value=&FloatFromCurve(
        FloatCurve=DataInterface() {propertyValues {Curve {keys [<at 0.1365 0>,<at 0.1754 0.6533>,<at 0.2838 0>,<at 0.345 0.2634>,<at 0.4303 0.149>,<at 0.5361 0.4843>,<at 0.5674 0>,<at 1.3326 0>,<at 1.3532 0.2307>,<at 1.4591 0>,<at 1.5391 0.6635>,<at 1.6457 0>,<at 1.7194 0.4348>,<at 1.7831 0>,<at 2.316 0>,<at 2.36 0.5083>,<at 2.4843 0>,<at 3 0>]}}}
        CurveIndex=&Add_Float(A=$Particles.NormalizedAge, B=&RandomRangeFloat(Minimum=-0.2, Maximum=2))
        Scale Curve=&RandomRangeFloat(Minimum=0.2, Maximum=0.6)
       ))
      "Float (Alpha)"=0
     )
    )
    Particles.GlowAlignment:Vector3f=<0 0 1>
    Particles.GlowOffset=-5
    Particles.GlowSize=&Multiply_Vector2D(
     A=&MakeVector2D(
      X=&MakeFloatFromVector(Vector=$Particles.Scale, Channel=::NewEnumerator2)
      Y=&MakeFloatFromVector(Vector=$Particles.Scale, Channel=::NewEnumerator0)
     )
     B=<20.8 50.3>
    )
   ) {isSetParameters=true, module=SetVariables_#17}
   &ScaleMeshSize(Scale Factor=&VectorFromFloat(Value=&FloatFromCurve(FloatCurve=DataInterface() {propertyValues {Curve {
         defaultValue=1
         keys [
          {arriveTangent=5.1268, interpMode=RCIM_Cubic, leaveTangent=5.1268, tangentMode=RCTM_User, time=0, value=0}
          {
           arriveTangentWeight=0.2927
           interpMode=RCIM_Cubic
           leaveTangentWeight=0.2927
           tangentMode=RCTM_User
           tangentWeightMode=RCTWM_WeightedBoth
           time=0.5
           value=1
          }
          {arriveTangent=-3.873, tangentMode=RCTM_User, time=1, value=0}
         ]
        }}})))
   &UpdateMeshOrientation(Rotation Rate=&RandomRangeFloat(Minimum=-2, Maximum=2), Rotation Vector=<0 0 1>)
  ]
  rendererClasses [&NiagaraSpriteRendererProperties, &NiagaraMeshRendererProperties]
  renderers [
   {rendererClass=&NiagaraSpriteRendererProperties}
   {rendererClass=&NiagaraSpriteRendererProperties, rendererIndex=1}
   {rendererClass=&NiagaraMeshRendererProperties, rendererIndex=2}
   {rendererClass=&NiagaraSpriteRendererProperties, rendererIndex=3}
  ]
 }
 Sparks_Slow {
  emitterUpdateScript [
   &EmitterState(
    Delay First Loop Only=true
    Life Cycle Mode=::NewEnumerator1
    Loop Delay=&Divide_Float(A=70, B=30)
    Loop Duration=&Divide_Float(A=40, B=90)
    UseLoopDelay=true
   )
   &SpawnBurst_Instantaneous(
    Spawn Count=&RandomRangeInt(Minimum=5, Maximum=8)
    Spawn Probability=&ClampFloat(Float=&Subtract_Float(A=$System.SpawnForward, B=$System.SpawnRegular))
    Spawn Time=&Divide_Float(A=2, B=90)
    Use Spawn Probability=true
    UseLoopCountLimit=false
   )
   &SpawnBurst_Instantaneous(
    Spawn Count=&RandomRangeInt(Minimum=5, Maximum=8)
    Spawn Group=1
    Spawn Probability=&ClampFloat(Float=&Subtract_Float(A=$System.SpawnForward, B=$System.SpawnRegular))
    Spawn Time=&Divide_Float(A=34, B=90)
    Use Spawn Probability=true
    UseLoopCountLimit=false
   ) {module=SpawnBurst_Instantaneous001}
   &SpawnBurst_Instantaneous(
    Spawn Count=&RandomRangeInt(Minimum=5, Maximum=8)
    Spawn Group=2
    Spawn Probability=&ClampFloat(Float=&Subtract_Float(A=$System.SpawnForward, B=$System.SpawnRegular))
    Spawn Time=&Divide_Float(A=26, B=90)
    Use Spawn Probability=true
    UseLoopCountLimit=false
   ) {module=SpawnBurst_Instantaneous002}
   &SpawnBurst_Instantaneous(
    Spawn Count=&RandomRangeInt(Minimum=5, Maximum=8)
    Spawn Group=3
    Spawn Probability=&ClampFloat(Float=&Subtract_Float(A=$System.SpawnForward, B=$System.SpawnRegular))
    Spawn Time=&Divide_Float(A=18, B=90)
    Use Spawn Probability=true
    UseLoopCountLimit=false
   ) {module=SpawnBurst_Instantaneous003}
   &SpawnRate(
    Spawn Probability=&ClampFloat(Float=&Subtract_Float(A=$System.SpawnForward, B=$System.SpawnRegular))
    SpawnGroup=4
    SpawnRate=33
    Use Spawn Probability=true
   )
  ]
  particleSpawnScript [
   &PartitionParticles(
    A:NiagaraFloat=&RandomRangeFloat(Minimum=0, Maximum=0.6)
    B=0.5
    Partition Mode:ENiagara_PartitionMode_v2=::NewEnumerator4
   )
   /SetVariables(Particles.VisibilityTag=$Particles.PartitionParticles.Partition) {isSetParameters=true, module=SetVariables_#16}
   &InitializeParticle(
    Color=&RandomRangeLinearColor(Minimum=<rgba 4 0.2435 0.2435 0.2435>, Maximum=<rgba 5 0.3478 0.7355 0.713>)
    Lifetime=0.8
    Mesh Scale Max=<0.85 0.85 0.95>
    Mesh Scale Min=<0.7 0.7 0.8>
    Mesh Scale Mode=::NewEnumerator2
    Position Mode=::NewEnumerator0
    Sprite Size Max=<20 20>
    Sprite Size Min=<10 10>
    Sprite Size Mode=::NewEnumerator2
    Write Material Random=true
   )
   &SkeletalMeshLocation(
    Bone / Socket Index=&MaskIntBySpawnGroup(
     Masked Int=&RandomRangeInt(Minimum=5, Maximum=6)
     Passthrough Int=&SpawnGroup()
     Spawn Group Mask=4
    )
    MaskSpawnGroup=false
    Mesh Sampling Type=::NewEnumerator5
    Orientation Sampling=::NewEnumerator1
    Skeletal Mesh=$System.Skeletal Mesh
    Socket Sampling Mode=::NewEnumerator5
   )
   &ShapeLocation(
    MaskSpawnGroup=false
    Override Seed=false
    Ring Radius=13
    Rotation Coordinate Space=::World
    Shape Primitive=::NewEnumerator4
    UseOwnerScale=true
   )
   &InitialMeshOrientation(
    Mesh Axis To Orient=<0 0 1>
    Mesh Orientation Mode=::NewEnumerator2
    Orientation Vector=<0 0 1>
    Rotation=&RandomRangeVector(Minimum=<0 0 0>, Maximum=<0 0 1>)
    Use Rotation Vector=true
   )
   &AddVelocity(
    Rotation Coordinate Space=::World
    Velocity=&AddVectorToPosition(
     Position=&RandomVector(Vector Scale=&MaskFloat(Mask Source=1, Mask Parameter=$Particles.PartitionParticles.Partition, Mask=0, Masked Float=0.5, Passthrough Float=1.6))
     Vector=&Multiply_VectorByFloat(
      Vector=<0 0 1>
      Float=&MaskFloat(
       Mask Source=1
       Mask Parameter=$Particles.PartitionParticles.Partition
       Mask=0
       Masked Float=&RandomRangeFloat(Minimum=0.6, Maximum=0.8)
       Passthrough Float=0
      )
     )
    )
    Velocity Speed Scale=70
   ) {module=AddVelocity001}
   &AddVelocity(Velocity=<0 -50 0>, Velocity Speed Scale=&FN_FrontEndFloatModifier(Gameplay=0, Front End=1))
  ]
  particleUpdateScript [
   &ParticleState()
   &Drag(
    Drag=&MaskFloat(Mask Source=1, Mask Parameter=$Particles.PartitionParticles.Partition, Mask=0, Masked Float=1.5, Passthrough Float=2)
    Rotational Drag=1.5
    Use Linear Drag=true
    Use Rotational Drag=true
   )
   &ScaleSpriteSize(
    Scale Sprite Size Mode=::NewEnumerator0
    Uniform Scale Factor=&Multiply_Float(
     A=&FloatFromCurve(FloatCurve=DataInterface() {propertyValues {Curve {
         defaultValue=1
         keys [
          {arriveTangent=6.4214, interpMode=RCIM_Cubic, leaveTangent=6.4214, tangentMode=RCTM_User, time=0, value=0}
          {interpMode=RCIM_Cubic, tangentMode=RCTM_User, time=0.2882, value=1}
          {arriveTangent=-2.7349, tangentMode=RCTM_User, time=1, value=0}
         ]
        }}})
     B=&FloatFromCurve(
      FloatCurve=DataInterface() {propertyValues {Curve {keys [<at 0 0.2>,<at 0.1 0.8647>,<at 0.7 1>,<at 0.85 1.3>]}}}
      CurveIndex=&RandomRangeFloat(Minimum=0, Maximum=1, Evaluation Type=::NewEnumerator1)
     )
    )
   )
   &ScaleColor(
    Linear Color Curve=DataInterface() {propertyValues {
      AlphaCurve {keys [<at 0 0>,<at 0.0716 0.2487>,<at 0.1723 1>,<at 0.504 0.6985>,<at 0.7379 0.2792>,<at 1 0>]}
      BlueCurve {keys [<at 0 1>,<at 1 1>]}
      GreenCurve {keys [<at 0 1>,<at 1 1>]}
      RedCurve {keys [<at 0 1>,<at 1 1>]}
     }}
    Scale Mode=2
   )
   &InheritSourceMovement(Applied Position Scale=<0.7 0.7 1>, ApplyForce=false, ApplyPosition=true, ApplyPositionFalloff=false)
   &SolveForcesAndVelocity(Clamp Velocity=false, Limit Acceleration=false)
   /SetVariables(
    Particles.Crystal Color:LinearColor=&Add_LinearColor(
     A:LinearColor=&ColorFromCurve(ColorCurve=DataInterface() {propertyValues {
        AlphaCurve {keys [<at 0.5781 1 RCIM_Cubic>,<at 1 0 RCIM_Cubic>]}
        BlueCurve {keys [<at 0 0 RCIM_Cubic>]}
        GreenCurve {keys [<at 0 0.0145 RCIM_Cubic>]}
        RedCurve {keys [<at 0 0.868 RCIM_Cubic>]}
       }})
     B:LinearColor=&MakeLinearColorFromVectorAndFloat(
      "Vector (RGB)"=&VectorFromFloat(Value=&FloatFromCurve(
        FloatCurve=DataInterface() {propertyValues {Curve {keys [<at 0.1365 0>,<at 0.1754 0.6533>,<at 0.2838 0>,<at 0.345 0.2634>,<at 0.4303 0.149>,<at 0.5361 0.4843>,<at 0.5674 0>,<at 1.3326 0>,<at 1.3532 0.2307>,<at 1.4591 0>,<at 1.5391 0.6635>,<at 1.6457 0>,<at 1.7194 0.4348>,<at 1.7831 0>,<at 2.316 0>,<at 2.36 0.5083>,<at 2.4843 0>,<at 3 0>]}}}
        CurveIndex=&Add_Float(A=$Particles.NormalizedAge, B=&RandomRangeFloat(Minimum=-0.2, Maximum=2))
        Scale Curve=&RandomRangeFloat(Minimum=0.2, Maximum=0.6)
       ))
      "Float (Alpha)"=0
     )
    )
    Particles.GlowAlignment:Vector3f=<0 0 1>
    Particles.GlowOffset=-5
    Particles.GlowSize=&Multiply_Vector2D(
     A=&MakeVector2D(
      X=&MakeFloatFromVector(Vector=$Particles.Scale, Channel=::NewEnumerator2)
      Y=&MakeFloatFromVector(Vector=$Particles.Scale, Channel=::NewEnumerator0)
     )
     B=<20.8 50.3>
    )
   ) {isSetParameters=true, module=SetVariables_#17}
   &ScaleMeshSize(Scale Factor=&VectorFromFloat(Value=&FloatFromCurve(FloatCurve=DataInterface() {propertyValues {Curve {
         defaultValue=1
         keys [
          {arriveTangent=5.1268, interpMode=RCIM_Cubic, leaveTangent=5.1268, tangentMode=RCTM_User, time=0, value=0}
          {
           arriveTangentWeight=0.2927
           interpMode=RCIM_Cubic
           leaveTangentWeight=0.2927
           tangentMode=RCTM_User
           tangentWeightMode=RCTWM_WeightedBoth
           time=0.5
           value=1
          }
          {arriveTangent=-3.873, tangentMode=RCTM_User, time=1, value=0}
         ]
        }}})))
   &UpdateMeshOrientation(Rotation Rate=&RandomRangeFloat(Minimum=-2, Maximum=2), Rotation Vector=<0 0 1>)
  ]
  rendererClasses [&NiagaraSpriteRendererProperties, &NiagaraMeshRendererProperties]
  renderers [
   {rendererClass=&NiagaraSpriteRendererProperties}
   {rendererClass=&NiagaraSpriteRendererProperties, rendererIndex=1}
   {rendererClass=&NiagaraMeshRendererProperties, rendererIndex=2}
   {rendererClass=&NiagaraSpriteRendererProperties, rendererIndex=3}
  ]
 }
 Trail_Green {
  emitterSpawnScript [&FN_RibbonIDOffsetUtility()]
  emitterUpdateScript [
   &EmitterState()
   &FN_SpawnAnimTrail(Number of Anim Trails=2)
  ]
  particleSpawnScript [
   &FN_InitializeAnimTrailBoneIndecies(First Bone or Socket Index=5, Second Bone or Socket Index=4)
   &FN_InitializeAnimTrailBoneIndecies(AnimTrail Selection=1, First Bone or Socket Index=6, Second Bone or Socket Index=4) {module=FN_InitializeAnimTrailBoneIndecies001}
   &FN_InitializeAnimTrail(Color=<rgba 1 0.0609 0.085 0.8>, Lifetime=0.16)
   &AddVelocity(Velocity=&FN_FrontEndVector3Modifier(Gameplay=<0 -2500 -750>, Front End=<0 2500 0>))
  ]
  particleUpdateScript [
   &InheritSourceMovement(
    Applied Position Scale=&Multiply_VectorByFloat(
     Vector=<1 1 1>
     Float=&FloatFromCurve(FloatCurve=DataInterface() {propertyValues {Curve {keys [
          {interpMode=RCIM_Cubic, tangentMode=RCTM_User, time=0.66, value=1}
          {arriveTangent=-1.9042, interpMode=RCIM_Cubic, leaveTangent=-1.9042, tangentMode=RCTM_User, time=1, value=0.5}
         ]}}})
    )
    ApplyForce=false
    ApplyPosition=true
    ApplyPositionFalloff=false
   )
   &SolveForcesAndVelocity(Clamp Velocity=false, Limit Acceleration=false)
   &FN_UpdateAnimTrail()
   &InterpolateOverTime(Initial Float Value:NiagaraFloat=$System.Timeline.Time, Target Float Value=1, Use Set Interpolation Rate=true) {enabled=false}
   &ScaleColor(Scale Mode=3)
   &ScaleColor(
    Scale Alpha=&Multiply_Float(
     A=&FN_FrontEndFloatModifier(Gameplay=$System.Timeline.Time, Front End=1)
     B=&FloatFromCurve(FloatCurve=DataInterface() {propertyValues {Curve {keys [<at 2.3333 0>,<at 3 1>]}}}, CurveIndex=$Engine.System.Age)
    )
    ScaleA=true
    ScaleRGB=true
   ) {module=ScaleColor003}
   /SetVariables(
    Particles.TipColor:LinearColor=&MultiplyLinearColor(Linear Color=$Particles.Color, Scale Factor=<1 1 1 1>)
    Particles.WingColor:LinearColor=&MultiplyLinearColor(
     Linear Color=$Particles.Color
     Scale Factor=&MakeVector4(
      X=1
      Y=1
      Z=1
      W=&FloatFromCurve(FloatCurve=DataInterface() {propertyValues {Curve {keys [<at 0 0>,<at 0.05 1>]}}})
     )
    )
   ) {isSetParameters=true}
  ]
  rendererClasses [&NiagaraRibbonRendererProperties]
  renderers [
   {rendererClass=&NiagaraRibbonRendererProperties}
   {rendererClass=&NiagaraRibbonRendererProperties, rendererIndex=1}
  ]
 }
}
summary {
 systemName=NS_Glider_NobleCrusher_Gallop
 emitters [
  {emitterName=Flares_Reg, rendererClasses [&NiagaraSpriteRendererProperties, &NiagaraMeshRendererProperties]}
  {bEnabled=false, emitterName=CapsuleGlows001, rendererClasses [&NiagaraSpriteRendererProperties]}
  {bEnabled=false, emitterName=CapsuleGlows002, rendererClasses [&NiagaraSpriteRendererProperties]}
  {
   bEnabled=false
   emitterName=CapsuleGlows004
   rendererClasses [&NiagaraSpriteRendererProperties, &NiagaraMeshRendererProperties]
  }
  {
   bEnabled=false
   emitterName=CapsuleGlows005
   rendererClasses [&NiagaraSpriteRendererProperties, &NiagaraMeshRendererProperties]
  }
  {emitterName=Sparks_Reg, rendererClasses [&NiagaraSpriteRendererProperties, &NiagaraMeshRendererProperties]}
  {
   bEnabled=false
   emitterName=CapsuleGlows008
   rendererClasses [&NiagaraSpriteRendererProperties, &NiagaraMeshRendererProperties]
  }
  {bEnabled=false, emitterName=CapsuleGlows009, rendererClasses [&NiagaraSpriteRendererProperties]}
  {
   bEnabled=false
   emitterName=Flares_Fwd001
   rendererClasses [&NiagaraSpriteRendererProperties, &NiagaraMeshRendererProperties]
  }
  {
   bEnabled=false
   emitterName=Sparks_Fwd001
   rendererClasses [&NiagaraSpriteRendererProperties, &NiagaraMeshRendererProperties]
  }
  {emitterName=Flares_Slow, rendererClasses [&NiagaraSpriteRendererProperties, &NiagaraMeshRendererProperties]}
  {emitterName=Sparks_Slow, rendererClasses [&NiagaraSpriteRendererProperties, &NiagaraMeshRendererProperties]}
  {emitterName=Trail_Green, rendererClasses [&NiagaraRibbonRendererProperties]}
 ]
 userVariables [
  User.bIsFrontEnd {type=NiagaraBool}
  User.bIsFrontEndPreview {type=NiagaraBool}
  User.ForwardDot {type=NiagaraFloat, defaultValue=1, description=""}
  User.RibbonIDOffset {type=NiagaraInt32}
  User.RightDot {type=NiagaraFloat, defaultValue=0, description=""}
 ]
}
